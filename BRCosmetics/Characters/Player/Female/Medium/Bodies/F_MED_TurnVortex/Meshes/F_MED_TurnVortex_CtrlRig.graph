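# Control Rig for the F_MED_TurnVortex female cosmetic body, adding deformation correctives and positioning jacket-chain, necklace and pendant bones.
# It runs a shared DeformRig function, then jacketChains, then shortChain.
# jacketChains uses weighted parent constraints to make chain bases, back colliders, necklace and pendant follow coat and body bones.
# shortChain uses a left-thigh spherical pose reader to weight a pendant offset, choosing a scaled-down variant when isCrouching is true.
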ControlRigBlueprint /BRCosmetics/Characters/Player/Female/Medium/Bodies/F_MED_TurnVortex/Meshes/F_MED_TurnVortex_CtrlRig
graphs [
 New Function_ContainedGraph {nodes [
   RigVMFunctionEntryNode()
   > &RigUnit_ParentConstraint(
    AdvancedSettings=(RotationOrderForFilter=XZY)
    Child=(Type=Bone,Name="base_jacketChain_01_l")
    Parents=((Item=(Type=Bone,Name="dyn_coat_mid_01_L"),Weight=0.0073),(Item=(Type=Bone,Name="dyn_coat_fr_01_L"),Weight=0.1317),(Item=(Type=Bone,Name="dyn_coat_01_L"),Weight=0.2565),(Item=(Type=Bone,Name="dyn_coat_01_fL"),Weight=0.6046))
    Weight=1
   )
   > &RigUnit_ParentConstraint(
    AdvancedSettings=(RotationOrderForFilter=XZY)
    Child=(Type=Bone,Name="base_jacketChain_02_l")
    Parents="((Item=(Type=Bone,Name=\"dyn_coat_fr_01_L\"),Weight=0.0005),(Item=(Type=Bone,Name=\"dyn_coat_mid_01_R\"),Weight=0.0024),(Item=(Type=Bone,Name=\"clavicle_l\"),Weight=0.015),(Item=(Type=Bone,Name=\"dyn_coat_bk_01\"),Weight=0.0256),(Item=(Type=Bone,Name=\"dyn_coat_01_ ...(truncated)"
    Weight=1
   )
   > &RigUnit_ParentConstraint(
    AdvancedSettings=(RotationOrderForFilter=XZY)
    Child=(Type=Bone,Name="base_jacketChain_back_l")
    Parents="((Item=(Type=Bone,Name=\"spine_03\"),Weight=0.0192),(Item=(Type=Bone,Name=\"dyn_coat_01_L\"),Weight=0.0404),(Item=(Type=Bone,Name=\"upperarm_twist_01_l\"),Weight=0.0422),(Item=(Type=Bone,Name=\"dyn_coat_bk_01\"),Weight=0.0545),(Item=(Type=Bone,Name=\"spine_04\"),Wei ...(truncated)"
    Weight=1
   )
   > &RigUnit_ParentConstraint(
    AdvancedSettings=(RotationOrderForFilter=XZY)
    Child=(Type=Bone,Name="base_jacketChain_01_r")
    Parents=((Item=(Type=Bone,Name="dyn_coat_mid_01_R"),Weight=0.0073),(Item=(Type=Bone,Name="dyn_coat_fr_01_R"),Weight=0.1317),(Item=(Type=Bone,Name="dyn_coat_01_R"),Weight=0.2565),(Item=(Type=Bone,Name="dyn_coat_01_fR"),Weight=0.6046))
    Weight=1
   )
   > &RigUnit_ParentConstraint(
    AdvancedSettings=(RotationOrderForFilter=XZY)
    Child=(Type=Bone,Name="base_jacketChain_02_r")
    Parents="((Item=(Type=Bone,Name=\"dyn_coat_fr_01_R\"),Weight=0.0005),(Item=(Type=Bone,Name=\"dyn_coat_mid_01_L\"),Weight=0.0024),(Item=(Type=Bone,Name=\"clavicle_r\"),Weight=0.015),(Item=(Type=Bone,Name=\"dyn_coat_bk_01\"),Weight=0.0256),(Item=(Type=Bone,Name=\"dyn_coat_01_ ...(truncated)"
    Weight=1
   )
   > &RigUnit_ParentConstraint(
    AdvancedSettings=(RotationOrderForFilter=XZY)
    Child=(Type=Bone,Name="base_jacketChain_back_r")
    Parents="((Item=(Type=Bone,Name=\"spine_03\"),Weight=0.0192),(Item=(Type=Bone,Name=\"dyn_coat_01_R\"),Weight=0.0404),(Item=(Type=Bone,Name=\"upperarm_twist_01_r\"),Weight=0.0422),(Item=(Type=Bone,Name=\"dyn_coat_bk_01\"),Weight=0.0545),(Item=(Type=Bone,Name=\"spine_04\"),Wei ...(truncated)"
    Weight=1
   )
   > &RigUnit_ParentConstraint(
    AdvancedSettings=(RotationOrderForFilter=XZY)
    Child=(Type=Bone,Name="base_back_chainCollider_L")
    Parents="((Item=(Type=Bone,Name=\"spine_05\"),Weight=0.0021),(Item=(Type=Bone,Name=\"spine_04\"),Weight=0.0033),(Item=(Type=Bone,Name=\"dyn_coat_01_fL\"),Weight=0.0062),(Item=(Type=Bone,Name=\"clavicle_l\"),Weight=0.0091),(Item=(Type=Bone,Name=\"dyn_coat_mid_01_R\"),Weight=0 ...(truncated)"
    Weight=1
   )
   > &RigUnit_ParentConstraint(
    AdvancedSettings=(RotationOrderForFilter=XZY)
    Child=(Type=Bone,Name="base_back_chainCollider_R")
    Parents="((Item=(Type=Bone,Name=\"spine_05\"),Weight=0.0021),(Item=(Type=Bone,Name=\"spine_04\"),Weight=0.0033),(Item=(Type=Bone,Name=\"dyn_coat_01_fR\"),Weight=0.0062),(Item=(Type=Bone,Name=\"clavicle_r\"),Weight=0.0091),(Item=(Type=Bone,Name=\"dyn_coat_mid_01_L\"),Weight=0 ...(truncated)"
    Weight=1
   )
   > &RigUnit_ParentConstraint(
    AdvancedSettings=(RotationOrderForFilter=XZY)
    Child=(Type=Bone,Name="base_necklace")
    Parents=((Item=(Type=Bone,Name="FACIAL_C_UnderChin"),Weight=0.0133),(Item=(Type=Bone,Name="spine_05"),Weight=0.0564),(Item=(Type=Bone,Name="head"),Weight=0.2684),(Item=(Type=Bone,Name="neck_01"),Weight=0.3107),(Item=(Type=Bone,Name="neck_02"),Weight=0.3512))
    Weight=1
   )
   > &RigUnit_ParentConstraint(
    AdvancedSettings=(RotationOrderForFilter=XZY)
    Child=(Type=Bone,Name="base_pendant")
    Parents=((Item=(Type=Bone,Name="spine_02"),Weight=0.456),(Item=(Type=Bone,Name="pelvis"),Weight=0.544))
    Weight=1
   )
   > RigVMFunctionReturnNode()
   &RigUnit_ParentConstraint(
    AdvancedSettings=(RotationOrderForFilter=XZY)
    Child=(Type=Bone,Name="base_necklace")
    ExecutePin=()
    Parents=((Item=(Type=Bone,Name="FACIAL_C_UnderChin"),Weight=0.0133),(Item=(Type=Bone,Name="spine_05"),Weight=0.0564),(Item=(Type=Bone,Name="head"),Weight=0.2684),(Item=(Type=Bone,Name="neck_01"),Weight=0.3107),(Item=(Type=Bone,Name="neck_02"),Weight=0.3512))
    Weight=1
   )
  ]}
 New Function_ContainedGraph {nodes [
   n1 = RigVMFunctionEntryNode()
   RigVMFunction_Sequence = &RigVMFunction_Sequence(ExecuteContext=@n1.ExecuteContext)
   n2 = &RigUnit_SphericalPoseReader(
    ActiveRegionScaleFactors=(PositiveWidth=1,NegativeWidth=1,PositiveHeight=1,NegativeHeight=1)
    ActiveRegionSize=0.05
    Debug=(bDrawDebug=false,bDraw2D=False,bDrawLocalAxes=False,DebugScale=25,DebugSegments=20,DebugThickness=0.25)
    DriverAxis=(X=-1,Y=0,Z=0)
    DriverItem=(Type=Bone,Name="thigh_l")
    FalloffRegionScaleFactors=(PositiveWidth=1,NegativeWidth=1,PositiveHeight=1,NegativeHeight=1)
    FalloffSize=0.45
    FlipHeightScaling=False
    FlipWidthScaling=false
    OptionalParentItem=(Type=Bone,Name="None")
    RotationOffset=(X=90,Y=0,Z=-20)
    ExecutePin=@RigVMFunction_Sequence.A
   )
   RigVMFunction_ControlFlowBranch = &RigVMFunction_ControlFlowBranch(ExecuteContext=@n2.ExecutePin, Condition=RigVMVariableNode() {variable=isCrouching}.Value)
   &RigUnit_ModifyTransforms(
    ItemToModify=((Item=(Type=Bone,Name="base_pendant"),Transform=(Rotation=(X=-0.2257,Y=-0.4484,Z=0.1673,W=0.8486),Translation=(X=0,Y=0,Z=0),Scale3D=(X=1,Y=1,Z=1))))
    WeightMinimum=0
    ExecutePin=@RigVMFunction_ControlFlowBranch.False
    Weight=@n2.OutputParam
   )
   &RigUnit_ModifyTransforms(
    ItemToModify=((Item=(Type=Bone,Name="base_pendant"),Transform=(Rotation=(X=0.368,Y=0.2385,Z=-0.3662,W=-0.8207),Translation=(X=0,Y=0,Z=0),Scale3D=(X=0.7,Y=0.7,Z=0.7))))
    WeightMinimum=0
    ExecutePin=@RigVMFunction_ControlFlowBranch.True
    Weight=@n2.OutputParam
   )
   RigVMFunctionReturnNode(ExecuteContext=@RigVMFunction_Sequence.B)
  ]}
 RigVMFunctionLibrary {nodes [
   jacketChains = RigVMCollapseNode(ExecuteContext=())
   shortChain = RigVMCollapseNode(ExecuteContext=())
  ]}
 RigVMModel {nodes [
   RigUnit_BeginExecution = &RigUnit_BeginExecution()
   FN_DeformRig_FNC = RigVMFunctionReferenceNode(
    DeformRig_ Curve="(deformRig_l_upperarm_up_alpha_72_#1=False,deformRig_r_upperarm_up_alpha_73_#2=False,deformRig_l_upperarm_fr_alpha_74_#3=False,deformRig_r_upperarm_fr_alpha_75_CB51B8 ...(truncated)"
    DeformRig_Bool="(upperarm_1_#4=True,elbow_3_#5=True,wrist_7_#6=True,groin_8_#7=True,glute_17_#8=True,knee_9_40C408324CA261 ...(truncated)"
    DeformRig_Debug=(upperarm_debug_18_#4=False,wrist_debug_20_#6=False,lat_debug_26_#9=False,groinGlute_debug_27_#7=False)
    DeformRig_Value="(upperarm_up_18_#10=(Rotation=(X=0.0,Y=-0.0,Z=0.0,W=1.0),Translation=(X=0,Y=0,Z=3),Scale3D=(X=0.70,Y=1.0,Z=1.0)),upperarm_fwd_fr_19_#11=(Rotation=(X=0.0,Y=-0.0,Z=0.0,W=1.0),Translation=(X=-5,Y=10,Z= ...(truncated)"
    ExecuteContext=@RigUnit_BeginExecution.ExecutePin
   )
   > jacketChains = RigVMFunctionReferenceNode()
   > shortChain = RigVMFunctionReferenceNode()
  ]}
]
hierarchy [
 root {loc=<0 0 0>, rot=<0 0 0 1>}
 pelvis {parent=root, loc=<0 -0.5301 98.0571>, rot=<0.0127 0.707 -0.0127 -0.707>}
 spine_01 {parent=pelvis, loc=<6.1208 -0.9328 0>, rot=<0 0 0.034 -0.9994>}
 spine_02 {parent=spine_01, loc=<6.5997 -0.2095 0>, rot=<0 0 0.0238 -0.9997>}
 spine_03 {parent=spine_02, loc=<6.4914 -0.1237 0>, rot=<0 0 0.0149 -0.9999>}
 spine_04 {parent=spine_03, loc=<6.9598 0.0794 0>, rot=<0 0 -0.0132 -0.9999>}
 spine_05 {parent=spine_04, loc=<6.9987 0 0>, rot=<0 0 -0.0015 -1>}
 neck_01 {parent=spine_05, loc=<11.3138 0.1529 0>, rot=<0 0 -0.2003 -0.9797>}
 neck_02 {parent=neck_01, loc=<4.5627 0 0>, rot=<0 0 0.0504 -0.9987>}
 head {parent=neck_02, loc=<5.3914 0 0>, rot=<0 0 0.115 -0.9934>}
 FACIAL_C_FacialRoot {parent=head, loc=<2.1524 -0.746 -0.0002>, rot=<0.4879 -0.5118 -0.4879 -0.5118>}
 FACIAL_C_Jaw {parent=FACIAL_C_FacialRoot, loc=<-0.0008 2.4782 1.9212>, rot=<0.3134 -0.0004 -0.0001 -0.9496>}
 FACIAL_C_Jawline {parent=FACIAL_C_Jaw, loc=<-0.0023 1.9183 7.0593>, rot=<0.2586 0 0 -0.966>}
 FACIAL_L_ChinSide {parent=FACIAL_C_Jaw, loc=<2.5319 0.1917 6.8344>, rot=<-0.2208 -0.3738 0.0499 -0.8995>}
 FACIAL_R_ChinSide {parent=FACIAL_C_Jaw, loc=<-2.5413 0.1859 6.8345>, rot=<-0.2208 0.3738 -0.0499 -0.8995>}
 FACIAL_C_Chin {parent=FACIAL_C_Jaw, loc=<-0.0033 0.5355 7.1467>, rot=<-0.2591 0 0 -0.9659>}
 FACIAL_C_UnderChin {parent=FACIAL_C_FacialRoot, loc=<0.0015 8.0522 4.3291>, rot=<0.4734 -0.0003 -0.0002 -0.8809>}
 FACIAL_L_UnderChin {parent=FACIAL_C_FacialRoot, loc=<2.3451 6.9234 2.9674>, rot=<0.3769 -0.3575 0.0714 -0.8515>}
 FACIAL_R_UnderChin {parent=FACIAL_C_FacialRoot, loc=<-2.349 6.9257 2.9602>, rot=<0.3769 0.3569 -0.0716 -0.8517>}
 FACIAL_L_Jawline {parent=FACIAL_C_FacialRoot, loc=<4.0993 5.1894 4.2571>, rot=<0.18 0.159 -0.0499 -0.9695>}
 FACIAL_R_Jawline {parent=FACIAL_C_FacialRoot, loc=<-4.1054 5.1913 4.2605>, rot=<0.18 -0.1597 0.0498 -0.9693>}
 base_necklace {parent=neck_01, loc=<3.623 5.3903 0>, rot=<0.1137 0.6979 0.6979 0.1137>}
 dyn_necklace {parent=base_necklace, loc=<0 0 0>, rot=<0 0 0 1>}
 clavicle_l {parent=spine_05, loc=<5.4831 -0.1053 -3.2946>, rot=<0.0465 -0.7289 -0.009 -0.6829>}
 upperarm_l {parent=clavicle_l, loc=<9.8658 -1.2136 -0.4105>, rot=<0.0164 -0.3245 -0.0269 -0.9454>}
 lowerarm_l {parent=upperarm_l, loc=<26.465 -0.8345 0.259>, rot=<0.0036 0.0161 -0.2298 -0.9731>}
 hand_l {parent=lowerarm_l, loc=<22.7355 -0.3203 -0.6105>, rot=<-0.6784 -0.0223 0.0735 -0.7307>}
 thumb_01_l {parent=hand_l, loc=<1.3118 -0.3191 -2.459>, rot=<0.6546 -0.4042 0.0664 -0.6354>}
 thumb_02_l {parent=thumb_01_l, loc=<5.9336 -0.0448 0.0401>, rot=<-0.0037 0.0213 0.298 -0.9543>}
 thumb_03_l {parent=thumb_02_l, loc=<2.2223 0.3682 -0.0907>, rot=<0.0118 -0.013 0.0066 -0.9998>}
 index_metacarpal_l {parent=hand_l, loc=<1.6535 0.0183 -2.96>, rot=<0.0158 -0.0124 -0.0173 -0.9996>}
 index_01_l {parent=index_metacarpal_l, loc=<8.3008 0.0357 -0.0524>, rot=<-0.0044 0.0485 0.0924 -0.9945>}
 index_02_l {parent=index_01_l, loc=<3.3587 -0.3004 0.0452>, rot=<-0.002 0.0027 0.3225 -0.9466>}
 index_03_l {parent=index_02_l, loc=<2.5578 -0.123 -0.0252>, rot=<-0.002 0.0067 0.2558 -0.9667>}
 middle_metacarpal_l {parent=hand_l, loc=<2.1856 -0.1725 -0.5984>, rot=<0.006 0.0168 -0.0329 -0.9993>}
 middle_01_l {parent=middle_metacarpal_l, loc=<7.3275 -0.2447 0.0129>, rot=<-0.0208 0.0042 0.0858 -0.9961>}
 middle_02_l {parent=middle_01_l, loc=<4.0123 0.1401 0.1175>, rot=<-0.0218 0.0258 0.3512 -0.9357>}
 middle_03_l {parent=middle_02_l, loc=<2.6077 0.0534 0.1373>, rot=<0.0082 0.0188 0.1708 -0.9851>}
 ring_metacarpal_l {parent=hand_l, loc=<2.2074 -0.3625 1.3638>, rot=<-0.0284 0.0722 -0.0458 -0.9959>}
 ring_01_l {parent=ring_metacarpal_l, loc=<7.0295 -0.6197 -0.1446>, rot=<-0.0489 -0.0252 0.1892 -0.9804>}
 ring_02_l {parent=ring_01_l, loc=<3.6174 0.4647 -0.1027>, rot=<0 0 0.2806 -0.9598>}
 ring_03_l {parent=ring_02_l, loc=<2.4152 0.114 0.0199>, rot=<0 0 0.1189 -0.9929>}
 pinky_metacarpal_l {parent=hand_l, loc=<1.7616 -0.9753 3.3814>, rot=<-0.0397 0.0814 -0.069 -0.9935>}
 pinky_01_l {parent=pinky_metacarpal_l, loc=<7.1709 -0.4831 -0.1094>, rot=<-0.0654 -0.0035 0.2242 -0.9723>}
 pinky_02_l {parent=pinky_01_l, loc=<2.8274 0.0893 -0.1043>, rot=<0.0037 0.0011 0.2854 -0.9584>}
 pinky_03_l {parent=pinky_02_l, loc=<2.2046 0.0307 -0.0078>, rot=<-0.0002 0.0098 0.1445 -0.9895>}
 deform_wrist_dn_base_l {parent=hand_l, loc=<0 -1.5326 0>, rot=<0 0 0 -1>}
 deform_wrist_dn_l {parent=deform_wrist_dn_base_l, loc=<0 0 0>, rot=<0 0 0 1>}
 deform_wrist_up_base_l {parent=hand_l, loc=<0 2.3009 0>, rot=<0 0 0 -1>}
 deform_wrist_up_l {parent=deform_wrist_up_base_l, loc=<0 0 0>, rot=<0 0 0 1>}
 lowerarm_twist_02_l {parent=lowerarm_l, loc=<10.6707 0 0>, rot=<0 0 0 1>}
 lowerarm_twist_01_l {parent=lowerarm_l, loc=<18.316 0 0>, rot=<0 0 0 1>}
 deform_elbow_in_base_l {parent=lowerarm_l, loc=<0.1834 3.6062 -0.2572>, rot=<0 0 0 -1>}
 deform_elbow_in_l {parent=deform_elbow_in_base_l, loc=<0 0 0>, rot=<0 0 0 1>}
 deform_elbow_out_base_l {parent=lowerarm_l, loc=<0 -4 0>, rot=<0 0 0 1>}
 deform_elbow_out_l {parent=deform_elbow_out_base_l, loc=<0 0 0>, rot=<0 0 0 1>}
 upperarm_twist_01_l {parent=upperarm_l, loc=<7.2135 -0.1134 0>, rot=<0 0 0 1>}
 deform_upperarm_bk_base_l {parent=upperarm_twist_01_l, loc=<-6.4053 -4.3746 0.314>, rot=<0 0 0 -1>}
 deform_upperarm_bk_l {parent=deform_upperarm_bk_base_l, loc=<0 0 0>, rot=<0 0 0 1>}
 deform_upperarm_fr_base_l {parent=upperarm_twist_01_l, loc=<-5.6312 5.092 0.3354>, rot=<0 0 0 -1>}
 deform_upperarm_fr_l {parent=deform_upperarm_fr_base_l, loc=<0 0 0>, rot=<0 0 0 1>}
 deform_upperarm_up_base_l {parent=upperarm_twist_01_l, loc=<-8.891 0.4825 3.9517>, rot=<0 0 0 -1>}
 deform_upperarm_up_l {parent=deform_upperarm_up_base_l, loc=<0 0 0>, rot=<0 0 0 1>}
 upperarm_twist_02_l {parent=upperarm_l, loc=<19.7998 -0.1134 0>, rot=<0 0 0 1>}
 clavicle_r {parent=spine_05, loc=<5.4831 -0.1053 3.2946>, rot=<0.7289 0.0465 -0.6829 0.009>}
 upperarm_r {parent=clavicle_r, loc=<-9.8658 1.2136 0.4105>, rot=<0.0164 -0.3245 -0.0269 -0.9454>}
 lowerarm_r {parent=upperarm_r, loc=<-26.465 0.8345 -0.259>, rot=<0.0036 0.0161 -0.2298 -0.9731>}
 hand_r {parent=lowerarm_r, loc=<-22.7355 0.3203 0.6106>, rot=<-0.6784 -0.0223 0.0735 -0.7307>}
 thumb_01_r {parent=hand_r, loc=<-1.3118 0.3191 2.459>, rot=<0.6546 -0.4042 0.0664 -0.6354>}
 thumb_02_r {parent=thumb_01_r, loc=<-5.9336 0.0448 -0.0401>, rot=<-0.0037 0.0213 0.298 -0.9543>}
 thumb_03_r {parent=thumb_02_r, loc=<-2.2223 -0.3682 0.0907>, rot=<0.0118 -0.013 0.0066 -0.9998>}
 index_metacarpal_r {parent=hand_r, loc=<-1.6535 -0.0183 2.96>, rot=<0.0158 -0.0124 -0.0173 -0.9996>}
 index_01_r {parent=index_metacarpal_r, loc=<-8.3008 -0.0357 0.0524>, rot=<-0.0044 0.0485 0.0924 -0.9945>}
 index_02_r {parent=index_01_r, loc=<-3.3587 0.3004 -0.0452>, rot=<-0.002 0.0027 0.3225 -0.9466>}
 index_03_r {parent=index_02_r, loc=<-2.5578 0.123 0.0252>, rot=<-0.002 0.0067 0.2558 -0.9667>}
 middle_metacarpal_r {parent=hand_r, loc=<-2.1856 0.1725 0.5984>, rot=<0.006 0.0168 -0.0329 -0.9993>}
 middle_01_r {parent=middle_metacarpal_r, loc=<-7.3275 0.2447 -0.0129>, rot=<-0.0208 0.0042 0.0858 -0.9961>}
 middle_02_r {parent=middle_01_r, loc=<-4.0123 -0.1401 -0.1175>, rot=<-0.0218 0.0258 0.3512 -0.9357>}
 middle_03_r {parent=middle_02_r, loc=<-2.6077 -0.0534 -0.1373>, rot=<0.0082 0.0188 0.1708 -0.9851>}
 ring_metacarpal_r {parent=hand_r, loc=<-2.2074 0.3625 -1.3638>, rot=<-0.0284 0.0722 -0.0458 -0.9959>}
 ring_01_r {parent=ring_metacarpal_r, loc=<-7.0295 0.6197 0.1446>, rot=<-0.0489 -0.0252 0.1892 -0.9804>}
 ring_02_r {parent=ring_01_r, loc=<-3.6174 -0.4647 0.1027>, rot=<0 0 0.2806 -0.9598>}
 ring_03_r {parent=ring_02_r, loc=<-2.4152 -0.114 -0.0199>, rot=<0 0 0.1189 -0.9929>}
 pinky_metacarpal_r {parent=hand_r, loc=<-1.7616 0.9753 -3.3814>, rot=<-0.0397 0.0814 -0.069 -0.9935>}
 pinky_01_r {parent=pinky_metacarpal_r, loc=<-7.1709 0.4831 0.1094>, rot=<-0.0654 -0.0035 0.2242 -0.9723>}
 pinky_02_r {parent=pinky_01_r, loc=<-2.8274 -0.0893 0.1043>, rot=<0.0037 0.0011 0.2854 -0.9584>}
 pinky_03_r {parent=pinky_02_r, loc=<-2.2046 -0.0307 0.0078>, rot=<-0.0002 0.0098 0.1445 -0.9895>}
 deform_wrist_dn_base_r {parent=hand_r, loc=<-0.0001 1.5328 0>, rot=<0 0 0 -1>}
 deform_wrist_dn_r {parent=deform_wrist_dn_base_r, loc=<0 0 0>, rot=<0 0 0 1>}
 deform_wrist_up_base_r {parent=hand_r, loc=<-0.0002 -2.3007 -0.0001>, rot=<0 0 0 -1>}
 deform_wrist_up_r {parent=deform_wrist_up_base_r, loc=<0 0 0>, rot=<0 0 0 1>}
 lowerarm_twist_02_r {parent=lowerarm_r, loc=<-10.6707 0 0>, rot=<0 0 0 -1>}
 lowerarm_twist_01_r {parent=lowerarm_r, loc=<-18.316 0 0>, rot=<0 0 0 -1>}
 deform_elbow_in_base_r {parent=lowerarm_r, loc=<-0.1834 -3.6062 0.2572>, rot=<0 0 0 -1>}
 deform_elbow_in_r {parent=deform_elbow_in_base_r, loc=<0 0 0>, rot=<0 0 0 1>}
 deform_elbow_out_base_r {parent=lowerarm_r, loc=<0 4 -0.0001>, rot=<0 0 0 -1>}
 deform_elbow_out_r {parent=deform_elbow_out_base_r, loc=<0 0 0>, rot=<0 0 0 1>}
 upperarm_twist_01_r {parent=upperarm_r, loc=<-7.2135 0.1134 0>, rot=<0 0 0 -1>}
 deform_upperarm_bk_base_r {parent=upperarm_twist_01_r, loc=<6.4053 4.3746 -0.314>, rot=<0 0 0 -1>}
 deform_upperarm_bk_r {parent=deform_upperarm_bk_base_r, loc=<0 0 0>, rot=<0 0 0 1>}
 deform_upperarm_fr_base_r {parent=upperarm_twist_01_r, loc=<5.6312 -5.092 -0.3354>, rot=<0 0 0 -1>}
 deform_upperarm_fr_r {parent=deform_upperarm_fr_base_r, loc=<0 0 0>, rot=<0 0 0 1>}
 deform_upperarm_up_base_r {parent=upperarm_twist_01_r, loc=<8.891 -0.4825 -3.9517>, rot=<0 0 0 -1>}
 deform_upperarm_up_r {parent=deform_upperarm_up_base_r, loc=<0 0 0>, rot=<0 0 0 1>}
 upperarm_twist_02_r {parent=upperarm_r, loc=<-19.7998 0.1134 0>, rot=<0 0 0 -1>}
 attach_backpack {parent=spine_05, loc=<-0.123 -13.473 0>, rot=<-0.0093 -0.707 0.0093 -0.707>}
 Backpack_BR {parent=attach_backpack, type=Null, loc=<0 -1.9568 0>, rot=<-0.0532 0 0 0.9986>}
 attach_cape {parent=spine_05, loc=<9.017 -11.714 0>, rot=<0.0616 -0.7044 -0.0616 -0.7044>}
 cape {parent=attach_cape, type=Null, loc=<0 0 -2.1587>, rot=<0 0 0 1>}
 deform_pec_aim_base_l {parent=spine_04, loc=<3.2456 9.3309 -1.2237>, rot=<0.0603 -0.4033 0.3318 -0.8507>}
 deform_pec_aim_l {parent=deform_pec_aim_base_l, loc=<0 0 0>, rot=<0 0 0 1>}
 deform_pec_base_l {parent=deform_pec_aim_l, loc=<9.3665 0 0>, rot=<0 0 0 1>}
 deform_pec_l {parent=deform_pec_base_l, loc=<0 0 0>, rot=<0 0 0 1>}
 deform_pec_aim_base_r {parent=spine_04, loc=<3.2452 9.3309 1.2237>, rot=<-0.4033 -0.0603 0.8507 0.3318>}
 deform_pec_aim_r {parent=deform_pec_aim_base_r, loc=<0 0 0>, rot=<0 0 0 1>}
 deform_pec_base_r {parent=deform_pec_aim_r, loc=<-9.3667 -0.0003 -0.0004>, rot=<0 0 0 1>}
 deform_pec_r {parent=deform_pec_base_r, loc=<0 0 0>, rot=<0 0 0 1>}
 base_jacketChain_01_l {parent=spine_04, loc=<-0.6261 4.6529 -10.9027>, rot=<0.0761 -0.9674 -0.2413 0.0096>}
 dyn_jacketChain_01_l {parent=base_jacketChain_01_l, loc=<0 0 0>, rot=<0 0 0 1>}
 base_jacketChain_02_l {parent=spine_04, loc=<0.1617 -4.587 -10.1515>, rot=<0.041 -0.8896 0.4547 0.008>}
 dyn_jacketChain_02_l {parent=base_jacketChain_02_l, loc=<0 0 0>, rot=<0 0 0 1>}
 base_jacketChain_back_l {parent=spine_04, loc=<4.0531 -6.6604 -6.5535>, rot=<0.0296 -0.7064 0.706 -0.0401>}
 dyn_jacketChain_back_01_l {parent=base_jacketChain_back_l, loc=<0 0 0>, rot=<0 0 0 1>}
 dyn_jacketChain_back_02_l {parent=dyn_jacketChain_back_01_l, loc=<6.1376 0 0>, rot=<0 0 0 1>}
 base_jacketChain_01_r {parent=spine_04, loc=<-0.6264 4.6528 10.9027>, rot=<0.9674 0.0761 0.0096 0.2413>}
 dyn_jacketChain_01_r {parent=base_jacketChain_01_r, loc=<0 0 0>, rot=<0 0 0 1>}
 base_jacketChain_02_r {parent=spine_04, loc=<0.1613 -4.5871 10.1515>, rot=<0.8896 0.041 0.008 -0.4547>}
 dyn_jacketChain_02_r {parent=base_jacketChain_02_r, loc=<0 0 0>, rot=<0 0 0 1>}
 base_jacketChain_back_r {parent=spine_04, loc=<4.0531 -6.6604 6.5535>, rot=<0.7064 0.0296 -0.0401 -0.706>}
 dyn_jacketChain_back_01_r {parent=base_jacketChain_back_r, loc=<0 0 0>, rot=<0 0 0 1>}
 dyn_jacketChain_back_02_r {parent=dyn_jacketChain_back_01_r, loc=<-6.1374 0 0>, rot=<0 0 0 1>}
 dyn_coat_fr_MSTR_L {parent=spine_04, loc=<5.0103 11.6614 -6.0438>, rot=<0.0552 -0.8946 -0.4418 0.0388>}
 dyn_coat_fr_01_L {parent=dyn_coat_fr_MSTR_L, loc=<1 0 0>, rot=<0 0 0 1>}
 dyn_coat_fr_02_L {parent=dyn_coat_fr_01_L, loc=<5.7086 -0.148 -0.4661>, rot=<0 0 0 1>}
 dyn_coat_MSTR_fL {parent=spine_04, loc=<4.4583 5.1442 -11.4602>, rot=<0.0437 -0.9764 -0.2112 0.0099>}
 dyn_coat_01_fL {parent=dyn_coat_MSTR_fL, loc=<1 0 0>, rot=<0 0 0 1>}
 dyn_coat_02_fL {parent=dyn_coat_01_fL, loc=<5.2518 0.6865 -0.8988>, rot=<0 0 0 1>}
 dyn_coat_MSTR_L {parent=spine_04, loc=<3.7061 -0.7398 -12.0705>, rot=<0.0447 -0.9966 0.0698 -0.0027>}
 dyn_coat_01_L {parent=dyn_coat_MSTR_L, loc=<1 0 0>, rot=<0 0 0 1>}
 dyn_coat_02_L {parent=dyn_coat_01_L, loc=<4.8223 0.3124 -0.4076>, rot=<0 0 0 1>}
 dyn_coat_mid_MSTR_L {parent=spine_04, loc=<4.0788 -6.1869 -6.4183>, rot=<0.0046 -0.7982 0.6013 0.0361>}
 dyn_coat_mid_01_L {parent=dyn_coat_mid_MSTR_L, loc=<1.0001 0 0>, rot=<-0.0567 0 0 -0.9984>}
 dyn_coat_mid_02_L {parent=dyn_coat_mid_01_L, loc=<5.5429 0 -0.2056>, rot=<0 0 0 -1>}
 dyn_coat_bk_MSTR {parent=spine_04, loc=<6.086 -6.2217 0.0329>, rot=<-0.0138 -0.6984 0.7155 0.011>}
 dyn_coat_bk_01 {parent=dyn_coat_bk_MSTR, loc=<1 0 0>, rot=<0 0 0 1>}
 dyn_coat_bk_02 {parent=dyn_coat_bk_01, loc=<7.7092 0.0159 0.8701>, rot=<0 0 0 1>}
 dyn_coat_fr_MSTR_R {parent=spine_04, loc=<5.0106 11.6615 6.0438>, rot=<-0.0429 0.5096 0.8578 -0.052>}
 dyn_coat_fr_01_R {parent=dyn_coat_fr_MSTR_R, loc=<1 0 0>, rot=<0 0 0 1>}
 dyn_coat_fr_02_R {parent=dyn_coat_fr_01_R, loc=<5.7094 0.0745 -0.4833>, rot=<0 0 0 1>}
 dyn_coat_MSTR_fR {parent=spine_04, loc=<4.4583 5.1442 11.4602>, rot=<-0.0098 0.2075 0.9772 -0.0437>}
 dyn_coat_01_fR {parent=dyn_coat_MSTR_fR, loc=<1 0 0>, rot=<0 0 0 1>}
 dyn_coat_02_fR {parent=dyn_coat_01_fR, loc=<5.2515 -0.6797 -0.9039>, rot=<0 0 0 1>}
 dyn_coat_MSTR_R {parent=spine_04, loc=<3.7058 -0.7398 12.0705>, rot=<0.0026 -0.0683 0.9967 -0.0447>}
 dyn_coat_01_R {parent=dyn_coat_MSTR_R, loc=<1 0 0>, rot=<0 0 0 1>}
 dyn_coat_02_R {parent=dyn_coat_01_R, loc=<4.8221 -0.3137 -0.4066>, rot=<0 0 0 1>}
 dyn_coat_mid_MSTR_R {parent=spine_04, loc=<4.0788 -6.1869 6.4183>, rot=<-0.0357 -0.6488 0.7601 -0.0068>}
 dyn_coat_mid_01_R {parent=dyn_coat_mid_MSTR_R, loc=<1 0 0>, rot=<0 0 0 1>}
 dyn_coat_mid_02_R {parent=dyn_coat_mid_01_R, loc=<5.5428 0.0017 -0.2056>, rot=<0 0 0 1>}
 base_back_chainCollider_L {parent=spine_04, loc=<-0.0132 -5.9594 -6.7279>, rot=<-0.02 -0.7626 0.6465 0.0132>}
 base_back_chainCollider_R {parent=spine_04, loc=<-0.0132 -5.9594 6.7279>, rot=<-0.0131 -0.6496 0.7599 0.0198>}
 thigh_l {parent=pelvis, loc=<-1.0855 -0.2109 -8.7053>, rot=<0.0009 0.0355 0.0155 -0.9993>}
 calf_l {parent=thigh_l, loc=<-42.5284 -0.0019 -0.7297>, rot=<0.0006 -0.0161 -0.0391 -0.9991>}
 foot_l {parent=calf_l, loc=<-46.7401 -2.5014 -0.592>, rot=<0.0427 -0.0238 0.0353 -0.9982>}
 ball_l {parent=foot_l, loc=<-3.5658 10.8066 -0.1544>, rot=<0.0001 -0.0001 -0.7063 -0.7079>}
 calf_twist_02_l {parent=calf_l, loc=<-19.1546 -1.7943 -0.2638>, rot=<0.0237 0.0095 -0.0035 -0.9997>}
 calf_twist_01_l {parent=calf_l, loc=<-35.8483 -1.9197 -0.5794>, rot=<0.0237 0.0095 -0.0035 -0.9997>}
 deform_knee_in_base_l {parent=calf_l, loc=<0.8031 -5.0231 0.2456>, rot=<0 0 0 -1>}
 deform_knee_in_l {parent=deform_knee_in_base_l, loc=<0 0 0>, rot=<0 0 0 1>}
 deform_knee_out_base_l {parent=calf_l, loc=<0 4 0>, rot=<0 0 0 -1>}
 deform_knee_out_l {parent=deform_knee_out_base_l, loc=<0 0 0>, rot=<0 0 0 1>}
 base_laces_l {parent=calf_l, loc=<-9.1695 2.7481 -0.5384>, rot=<0.043 0.7047 0.708 0.0156>}
 Base_rounfLaces_02_L {parent=base_laces_l, loc=<0.5848 0.7779 -0.0311>, rot=<0.2641 -0.0428 -0.1202 -0.956>}
 dyn_rounfLaces_02_L {parent=Base_rounfLaces_02_L, loc=<0 0 0>, rot=<0 0 0 1>}
 base_rounfLaces_01_l {parent=base_laces_l, loc=<0.1557 -1.1675 0.1245>, rot=<-0.1584 -0.0772 0.3628 -0.9151>}
 dyn_rounfLaces_01_l {parent=base_rounfLaces_01_l, loc=<0 0 0>, rot=<0 0 0 1>}
 base_lace_01_l {parent=base_laces_l, loc=<1.4031 -0.7572 -0.1669>, rot=<-0.124 -0.0354 0.0619 -0.9897>}
 dyn_lace_01_l {parent=base_lace_01_l, loc=<0 0 0>, rot=<0 0 0 1>}
 base_lace_02_l {parent=base_laces_l, loc=<1.3459 0.9826 -0.4872>, rot=<0.2557 -0.0774 0.0138 -0.9635>}
 dyn_lace_02_l {parent=base_lace_02_l, loc=<0 0 0>, rot=<0 0 0 1>}
 thigh_twist_01_l {parent=thigh_l, loc=<-27.952 0 0>, rot=<0.0013 0 0 -1>}
 deform_groin_base_l {parent=thigh_l, loc=<0.4801 5.6776 -0.2486>, rot=<0 0 0 -1>}
 deform_groin_l {parent=deform_groin_base_l, loc=<0 0 0>, rot=<0 0 0 1>}
 thigh_r {parent=pelvis, loc=<-1.0855 -0.2109 8.7053>, rot=<0.0355 -0.0009 0.9993 0.0155>}
 calf_r {parent=thigh_r, loc=<42.5284 0.0019 0.7297>, rot=<0.0006 -0.0161 -0.0391 -0.9991>}
 foot_r {parent=calf_r, loc=<46.7401 2.5014 0.592>, rot=<0.0427 -0.0238 0.0353 -0.9982>}
 ball_r {parent=foot_r, loc=<3.5658 -10.8066 0.1544>, rot=<0.0001 -0.0001 -0.7063 -0.7079>}
 calf_twist_02_r {parent=calf_r, loc=<19.1546 1.7943 0.2638>, rot=<0.0237 0.0095 -0.0035 -0.9997>}
 calf_twist_01_r {parent=calf_r, loc=<35.8483 1.9197 0.5794>, rot=<0.0237 0.0095 -0.0035 -0.9997>}
 deform_knee_in_base_r {parent=calf_r, loc=<-0.8031 5.0231 -0.2457>, rot=<0 0 0 -1>}
 deform_knee_in_r {parent=deform_knee_in_base_r, loc=<0 0 0>, rot=<0 0 0 1>}
 deform_knee_out_base_r {parent=calf_r, loc=<0 -4 0>, rot=<0 0 0 -1>}
 deform_knee_out_r {parent=deform_knee_out_base_r, loc=<0 0 0>, rot=<0 0 0 1>}
 base_races_r {parent=calf_r, loc=<9.1694 -2.7481 0.5384>, rot=<0.043 0.7047 0.708 0.0156>}
 base_rounfLaces_01_r {parent=base_races_r, loc=<-0.1557 1.1675 -0.1245>, rot=<-0.1584 -0.0772 0.3628 -0.9151>}
 dyn_rounfLaces_01_r {parent=base_rounfLaces_01_r, loc=<0 0 0>, rot=<0 0 0 -1>}
 Base_rounfLaces_02_r {parent=base_races_r, loc=<-0.5849 -0.7779 0.0311>, rot=<0.2641 -0.0428 -0.1202 -0.956>}
 dyn_rounfLaces_02_r {parent=Base_rounfLaces_02_r, loc=<0 0 0>, rot=<0 0 0 -1>}
 base_race_01_r {parent=base_races_r, loc=<-1.4032 0.7572 0.1669>, rot=<-0.124 -0.0354 0.0619 -0.9897>}
 dyn_race_01_r {parent=base_race_01_r, loc=<0 0 0>, rot=<0 0 0 1>}
 base_race_02_r {parent=base_races_r, loc=<-1.346 -0.9826 0.4872>, rot=<0.2557 -0.0774 0.0138 -0.9635>}
 dyn_race_02_r {parent=base_race_02_r, loc=<0 0 0>, rot=<0 0 0 1>}
 thigh_twist_01_r {parent=thigh_r, loc=<27.952 0 0>, rot=<0.0013 0 0 -1>}
 deform_groin_base_r {parent=thigh_r, loc=<-0.4801 -5.6776 0.2486>, rot=<0 0 0 -1>}
 deform_groin_r {parent=deform_groin_base_r, loc=<0 0 0>, rot=<0 0 0 1>}
 deform_glute_base_l {parent=pelvis, loc=<-7.3466 -9.0213 -9.1245>, rot=<0.0172 0.031 -0.4477 -0.8935>}
 deform_glute_l {parent=deform_glute_base_l, loc=<0 0 0>, rot=<0 0 0 1>}
 deform_glute_base_r {parent=pelvis, loc=<-7.3464 -9.0213 9.1245>, rot=<0.031 -0.0172 0.8935 -0.4477>}
 deform_glute_r {parent=deform_glute_base_r, loc=<0 0 0>, rot=<0 0 0 1>}
 base_frontChains {parent=pelvis, loc=<10.3275 4.8744 -7.1511>, rot=<-0.3141 -0.8077 -0.2844 0.4099>}
 dyn_frontChains_01 {parent=base_frontChains, loc=<0 0 0>, rot=<0 0 0 1>}
 dyn_frontChains_02 {parent=dyn_frontChains_01, loc=<6.2287 0 0>, rot=<0 -0.2473 0 -0.9689>}
 dyn_frontChains_03 {parent=dyn_frontChains_02, loc=<7.0109 0 0>, rot=<0 0 0 1>}
 base_backChains {parent=pelvis, loc=<12.0378 -5.7447 4.5139>, rot=<0.2821 -0.6052 0.6325 0.3926>}
 dyn_backChains_01 {parent=base_backChains, loc=<0 0 0>, rot=<0 0 0 1>}
 dyn_backChains_02 {parent=dyn_backChains_01, loc=<3.654 0 0>, rot=<0 -0.2942 0 -0.9558>}
 dyn_backChains_03 {parent=dyn_backChains_02, loc=<5.8384 0 0>, rot=<0 0 0 1>}
 base_pendant {parent=pelvis, loc=<10.7246 4.8596 -6.7069>, rot=<0.0015 -0.8445 -0.5296 -0.0799>}
 dyn_pendant {parent=base_pendant, loc=<0 0 0>, rot=<0 0 0 1>}
]
variables [isCrouching {type=bool, default=False}]
