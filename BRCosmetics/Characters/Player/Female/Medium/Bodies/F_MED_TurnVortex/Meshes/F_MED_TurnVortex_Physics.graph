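# Physics asset for the TurnVortex female medium player body cosmetic, setting up secondary motion for its coat, jacket chains and necklace.
# Capsules on the spine, neck, upper arms and back chain-collider bones form the body collision shapes.
# Coat segments have limited swing and twist against MSTR anchor bones, with fully free constraints linking adjacent side segments.
# Chains and necklace hang from base bones with limited swing and mostly locked twist.
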PhysicsAsset /BRCosmetics/Characters/Player/Female/Medium/Bodies/F_MED_TurnVortex/Meshes/F_MED_TurnVortex_Physics
bodies {
 base_back_chainCollider_L [{shapeType=Capsule, length=4, radius=2.5, shapeName=base_back_chainCollider_L_capsule, center=<1 0 -3>, rotation=<rot 0 0 98>}]
 base_back_chainCollider_R [{shapeType=Capsule, length=4, radius=2.5, shapeName=base_back_chainCollider_L_capsule, center=<1 0 -3>, rotation=<rot 0.5105 -2.8955 83.2>}]
 base_jacketChain_01_l [{
   shapeType=Capsule
   length=2.8297
   radius=1.689
   shapeName=base_jacketChain_01_l_capsule
   center=<1.1069 -0.0008 -1.9533>
   rotation=<rot 89.0233 -78.7693 -0.0055>
  }]
 base_jacketChain_01_r [{shapeType=Capsule, length=2, radius=0.7, shapeName=base_jacketChain_01_r_capsule, center=<-2 0 0>, rotation=<rot 90 0 0>}]
 base_jacketChain_02_l [{shapeType=Capsule, length=2, radius=0.7, shapeName=base_jacketChain_02_l_capsule, center=<-2 0 0>, rotation=<rot 90 0 0>}]
 base_jacketChain_02_r [{shapeType=Capsule, length=2, radius=0.7, shapeName=base_jacketChain_02_r_capsule, center=<-2 0 0>, rotation=<rot 90 0 0>}]
 base_jacketChain_back_l [{shapeType=Capsule, length=1, radius=1, shapeName=base_jacketChain_back_r_capsule}]
 base_jacketChain_back_r [{shapeType=Capsule, length=1, radius=1, shapeName=base_jacketChain_back_r_capsule}]
 base_necklace [{
   shapeType=Capsule
   length=0.505
   radius=0.505
   shapeName=base_necklace_capsule
   center=<-0.1106 0.0014 -0.0061>
   rotation=<rot 0 -89.8186 -82.1408>
  }]
 dyn_coat_01_L [{shapeType=Capsule, length=4, radius=3, shapeName=dyn_coat_01_L_capsule, center=<3 0 2>, rotation=<rot 0 0 90>}]
 dyn_coat_01_R [{shapeType=Capsule, length=4, radius=3, shapeName=dyn_coat_01_L_capsule, center=<3 0 2>, rotation=<rot 0 0 90>}]
 dyn_coat_01_fL [{shapeType=Capsule, length=4, radius=3, shapeName=dyn_coat_01_fL_capsule, center=<3 0 1.5>, rotation=<rot 0 0 89>}]
 dyn_coat_01_fR [{shapeType=Capsule, length=4, radius=3, shapeName=dyn_coat_01_fL_capsule, center=<3 0 1.5>, rotation=<rot 0 0 101>}]
 dyn_coat_MSTR_L [{
   shapeType=Capsule
   length=0.5319
   radius=0.5582
   shapeName=dyn_coat_MSTR_L_capsule
   center=<3.1804 0.0071 0.1575>
   rotation=<rot 0 0.0001 92.5648>
  }]
 dyn_coat_MSTR_R [{
   shapeType=Capsule
   length=0.5319
   radius=0.5582
   shapeName=dyn_coat_MSTR_R_capsule
   center=<3.1804 0.007 0.1575>
   rotation=<rot 0 -0.0001 92.5648>
  }]
 dyn_coat_MSTR_fL [{
   shapeType=Capsule
   length=0.5319
   radius=0.5582
   shapeName=dyn_coat_MSTR_fL_capsule
   center=<3.1804 0.0071 0.1575>
   rotation=<rot 0.0001 0.0001 92.5648>
  }]
 dyn_coat_MSTR_fR [{
   shapeType=Capsule
   length=0.5319
   radius=0.5582
   shapeName=dyn_coat_MSTR_fR_capsule
   center=<3.1804 0.007 0.1575>
   rotation=<rot 0.0001 -0.0001 92.5648>
  }]
 dyn_coat_bk_01 [{shapeType=Capsule, length=4, radius=3, shapeName=dyn_coat_bk_01_capsule, center=<5.0127 -0.0026 2.8932>, rotation=<rot 0 0 90>}]
 dyn_coat_bk_MSTR [{
   shapeType=Capsule
   length=0.5319
   radius=0.5582
   shapeName=dyn_coat_bk_MSTR_capsule
   center=<3.1804 0.0071 0.1575>
   rotation=<rot -0.0001 0 92.5648>
  }]
 dyn_coat_fr_01_L [{shapeType=Capsule, length=4, radius=3, shapeName=dyn_coat_fr_01_R_capsule, center=<3 -3 2>, rotation=<rot 0 0 90>}]
 dyn_coat_fr_01_R [{shapeType=Capsule, length=4, radius=3, shapeName=dyn_coat_fr_01_R_capsule, center=<3 2.5 1>, rotation=<rot 0.0004 -0.0002 103.4287>}]
 dyn_coat_fr_MSTR_L [{
   shapeType=Capsule
   length=0.5319
   radius=0.5582
   shapeName=dyn_coat_fr_MSTR_L_capsule
   center=<3.1804 0.0071 0.1575>
   rotation=<rot 0.0001 0.0001 92.5648>
  }]
 dyn_coat_fr_MSTR_R [{
   shapeType=Capsule
   length=0.5319
   radius=0.5582
   shapeName=dyn_coat_fr_MSTR_R_capsule
   center=<3.1804 0.0071 0.1575>
   rotation=<rot 0.0001 -0.0001 92.5648>
  }]
 dyn_coat_mid_01_L [{shapeType=Capsule, length=4, radius=3, shapeName=dyn_coat_mid_01_L_capsule, center=<3.013 0.0187 1.8929>, rotation=<rot 0 0 100>}]
 dyn_coat_mid_01_R [{shapeType=Capsule, length=4, radius=3, shapeName=dyn_coat_mid_01_L_capsule, center=<3.013 -0.0178 1.8928>, rotation=<rot 0 0 80>}]
 dyn_coat_mid_MSTR_L [{
   shapeType=Capsule
   length=0.5319
   radius=0.5582
   shapeName=dyn_coat_mid_MSTR_L_capsule
   center=<3.1804 0.0071 0.1575>
   rotation=<rot -0.0001 0 92.5648>
  }]
 dyn_coat_mid_MSTR_R [{
   shapeType=Capsule
   length=0.5319
   radius=0.5582
   shapeName=dyn_coat_mid_MSTR_R_capsule
   center=<3.1804 0.0071 0.1575>
   rotation=<rot -0.0001 0 92.5648>
  }]
 dyn_jacketChain_01_l [{shapeType=Capsule, length=2.8, radius=0.7, shapeName=dyn_jacketChain_01_l_capsule, center=<1.55 0 0.18>, rotation=<rot 90 0 0>}]
 dyn_jacketChain_01_r [{shapeType=Capsule, length=2.8, radius=0.7, shapeName=dyn_jacketChain_01_r_capsule, center=<-1.55 0 -0.18>, rotation=<rot 90 0 0>}]
 dyn_jacketChain_02_l [{shapeType=Capsule, length=2.5, radius=0.7, shapeName=dyn_jacketChain_02_l_capsule, center=<2.5 0 0.3>, rotation=<rot 90 0 0>}]
 dyn_jacketChain_02_r [{shapeType=Capsule, length=2.5, radius=0.7, shapeName=dyn_jacketChain_02_r_capsule, center=<-2.5 0 -0.3>, rotation=<rot 90 0 0>}]
 dyn_jacketChain_back_01_l [{shapeType=Capsule, length=4, radius=0.7, shapeName=dyn_jacketChain_back_01_l_capsule, center=<2.5 0 0>, rotation=<rot 90 0 0>}]
 dyn_jacketChain_back_01_r [{shapeType=Capsule, length=4, radius=0.7, shapeName=dyn_jacketChain_back_01_l_capsule, center=<-2.5 0 0>, rotation=<rot 90 0 0>}]
 dyn_jacketChain_back_02_l [{shapeType=Capsule, length=5, radius=0.7, shapeName=dyn_jacketChain_back_02_l_capsule, center=<2 0 0>, rotation=<rot 90 0 0>}]
 dyn_jacketChain_back_02_r [{shapeType=Capsule, length=5, radius=0.7, shapeName=dyn_jacketChain_back_02_l_capsule, center=<-2 0 0>, rotation=<rot 90 0 0>}]
 dyn_necklace [{shapeType=Capsule, length=2.0546, radius=0.9398, shapeName=dyn_necklace_capsule, center=<1.9811 0 0.5519>, rotation=<rot 83.299 -180 -90>}]
 neck_01 [{shapeType=Capsule, length=5, radius=4, shapeName=neck_01_capsule, center=<0.7073 1.8094 0>, rotation=<rot 90 -19.9082 0>}]
 neck_02 [{shapeType=Capsule, length=5, radius=4, shapeName=neck_02_capsule, center=<2.0222 0.7736 0>, rotation=<rot 90 -3.1862 0>}]
 spine_03 [{shapeType=Capsule, length=10.1122, radius=6.2858, shapeName=spine_03_capsule, center=<2.8903 2.4659 0>}]
 spine_04 [{shapeType=Capsule, length=4.2762, radius=7.4638, shapeName=spine_03_capsule, center=<0.4264 2.6395 0>}]
 spine_05 [{shapeType=Capsule, length=8.8258, radius=7.2439, center=<3.6418 -0.0132 0.0002>, rotation=<rot 0.0027 -90 0.0023>}]
 upperarm_l [{shapeType=Capsule, length=19.5586, radius=6.2442, center=<11.783 1.0221 2.9468>, rotation=<rot 84.7907 -179.9717 -179.9993>}]
 upperarm_r [{shapeType=Capsule, length=19.5586, radius=6.2442, center=<-10.4239 -1.4523 -1.5774>, rotation=<rot 84.7907 -179.9717 -179.9993>}]
}
constraints [
 {bone1Name=upperarm_l, bone2Name=spine_05, swing1LimitDegrees=45, swing2LimitDegrees=45, twistLimitDegrees=45}
 {bone1Name=upperarm_r, bone2Name=spine_05, swing1LimitDegrees=45, swing2LimitDegrees=45, twistLimitDegrees=45}
 {bone1Name=dyn_coat_mid_01_L, bone2Name=dyn_coat_mid_MSTR_L, swing1LimitDegrees=15, swing2LimitDegrees=50, twistLimitDegrees=1}
 {bone1Name=dyn_coat_mid_01_R, bone2Name=dyn_coat_mid_MSTR_R, swing1LimitDegrees=15, swing2LimitDegrees=50, twistLimitDegrees=1}
 {bone1Name=dyn_coat_fr_01_R, bone2Name=dyn_coat_fr_MSTR_R, swing1LimitDegrees=5, swing2LimitDegrees=20, twistLimitDegrees=1}
 {bone1Name=dyn_coat_fr_01_L, bone2Name=dyn_coat_fr_MSTR_L, swing1LimitDegrees=5, swing2LimitDegrees=20, twistLimitDegrees=1}
 {bone1Name=dyn_coat_bk_01, bone2Name=dyn_coat_bk_MSTR, swing1LimitDegrees=15, swing2LimitDegrees=50, twistLimitDegrees=1}
 {bone1Name=dyn_coat_01_fR, bone2Name=dyn_coat_MSTR_fR, swing1LimitDegrees=5, swing2LimitDegrees=20, twistLimitDegrees=1}
 {bone1Name=dyn_coat_01_fL, bone2Name=dyn_coat_MSTR_fL, swing1LimitDegrees=5, swing2LimitDegrees=20, twistLimitDegrees=1}
 {bone1Name=dyn_coat_01_R, bone2Name=dyn_coat_MSTR_R, swing1LimitDegrees=1, swing2LimitDegrees=7, twistLimitDegrees=1}
 {bone1Name=dyn_coat_01_L, bone2Name=dyn_coat_MSTR_L, swing1LimitDegrees=1, swing2LimitDegrees=7, twistLimitDegrees=1}
 {
  bone1Name=dyn_coat_MSTR_R
  bone2Name=spine_04
  swing1LimitDegrees=45
  swing1Motion=Free
  swing2LimitDegrees=45
  swing2Motion=Free
  twistLimitDegrees=45
  twistMotion=Free
 }
 {
  bone1Name=dyn_coat_MSTR_L
  bone2Name=spine_04
  swing1LimitDegrees=45
  swing1Motion=Free
  swing2LimitDegrees=45
  swing2Motion=Free
  twistLimitDegrees=45
  twistMotion=Free
 }
 {
  bone1Name=dyn_coat_MSTR_fL
  bone2Name=spine_04
  swing1LimitDegrees=45
  swing1Motion=Free
  swing2LimitDegrees=45
  swing2Motion=Free
  twistLimitDegrees=45
  twistMotion=Free
 }
 {
  bone1Name=dyn_coat_MSTR_fR
  bone2Name=spine_04
  swing1LimitDegrees=45
  swing1Motion=Free
  swing2LimitDegrees=45
  swing2Motion=Free
  twistLimitDegrees=45
  twistMotion=Free
 }
 {
  bone1Name=dyn_coat_bk_MSTR
  bone2Name=spine_04
  swing1LimitDegrees=45
  swing1Motion=Free
  swing2LimitDegrees=45
  swing2Motion=Free
  twistLimitDegrees=45
  twistMotion=Free
 }
 {
  bone1Name=dyn_coat_fr_MSTR_L
  bone2Name=spine_04
  swing1LimitDegrees=45
  swing1Motion=Free
  swing2LimitDegrees=45
  swing2Motion=Free
  twistLimitDegrees=45
  twistMotion=Free
 }
 {
  bone1Name=dyn_coat_fr_MSTR_R
  bone2Name=spine_04
  swing1LimitDegrees=45
  swing1Motion=Free
  swing2LimitDegrees=45
  swing2Motion=Free
  twistLimitDegrees=45
  twistMotion=Free
 }
 {
  bone1Name=dyn_coat_mid_MSTR_L
  bone2Name=spine_04
  swing1LimitDegrees=45
  swing1Motion=Free
  swing2LimitDegrees=45
  swing2Motion=Free
  twistLimitDegrees=45
  twistMotion=Free
 }
 {
  bone1Name=dyn_coat_mid_MSTR_R
  bone2Name=spine_04
  swing1LimitDegrees=45
  swing1Motion=Free
  swing2LimitDegrees=45
  swing2Motion=Free
  twistLimitDegrees=45
  twistMotion=Free
 }
 {
  bone1Name=dyn_coat_01_L
  bone2Name=dyn_coat_mid_01_L
  swing1LimitDegrees=45
  swing1Motion=Free
  swing2LimitDegrees=45
  swing2Motion=Free
  twistLimitDegrees=45
  twistMotion=Free
 }
 {
  bone1Name=dyn_coat_01_R
  bone2Name=dyn_coat_mid_01_R
  swing1LimitDegrees=45
  swing1Motion=Free
  swing2LimitDegrees=45
  swing2Motion=Free
  twistLimitDegrees=45
  twistMotion=Free
 }
 {
  bone1Name=dyn_coat_01_fL
  bone2Name=dyn_coat_01_L
  swing1LimitDegrees=45
  swing1Motion=Free
  swing2LimitDegrees=45
  swing2Motion=Free
  twistLimitDegrees=45
  twistMotion=Free
 }
 {
  bone1Name=dyn_coat_01_fR
  bone2Name=dyn_coat_01_R
  swing1LimitDegrees=45
  swing1Motion=Free
  swing2LimitDegrees=45
  swing2Motion=Free
  twistLimitDegrees=45
  twistMotion=Free
 }
 {
  bone1Name=dyn_coat_fr_01_L
  bone2Name=dyn_coat_01_fL
  swing1LimitDegrees=45
  swing1Motion=Free
  swing2LimitDegrees=45
  swing2Motion=Free
  twistLimitDegrees=45
  twistMotion=Free
 }
 {
  bone1Name=dyn_coat_fr_01_R
  bone2Name=dyn_coat_01_fR
  swing1LimitDegrees=45
  swing1Motion=Free
  swing2LimitDegrees=45
  swing2Motion=Free
  twistLimitDegrees=45
  twistMotion=Free
 }
 {
  bone1Name=base_jacketChain_01_l
  bone2Name=spine_04
  swing1LimitDegrees=45
  swing1Motion=Free
  swing2LimitDegrees=45
  swing2Motion=Free
  twistLimitDegrees=45
  twistMotion=Free
 }
 {
  bone1Name=dyn_jacketChain_01_l
  bone2Name=base_jacketChain_01_l
  swing1LimitDegrees=20
  swing2LimitDegrees=30
  twistLimitDegrees=0
  twistMotion=Locked
 }
 {
  bone1Name=base_jacketChain_01_r
  bone2Name=spine_04
  swing1LimitDegrees=45
  swing1Motion=Free
  swing2LimitDegrees=45
  swing2Motion=Free
  twistLimitDegrees=45
  twistMotion=Free
 }
 {bone1Name=dyn_jacketChain_01_r, bone2Name=base_jacketChain_01_r, swing1LimitDegrees=20, swing2LimitDegrees=30, twistLimitDegrees=10}
 {
  bone1Name=base_jacketChain_02_l
  bone2Name=spine_04
  swing1LimitDegrees=45
  swing1Motion=Free
  swing2LimitDegrees=45
  swing2Motion=Free
  twistLimitDegrees=45
  twistMotion=Free
 }
 {
  bone1Name=dyn_jacketChain_02_l
  bone2Name=base_jacketChain_02_l
  swing1LimitDegrees=20
  swing2LimitDegrees=30
  twistLimitDegrees=10
  twistMotion=Locked
 }
 {
  bone1Name=base_jacketChain_02_r
  bone2Name=spine_04
  swing1LimitDegrees=45
  swing1Motion=Free
  swing2LimitDegrees=45
  swing2Motion=Free
  twistLimitDegrees=45
  twistMotion=Free
 }
 {
  bone1Name=dyn_jacketChain_02_r
  bone2Name=base_jacketChain_02_r
  swing1LimitDegrees=20
  swing2LimitDegrees=30
  twistLimitDegrees=10
  twistMotion=Locked
 }
 {
  bone1Name=base_jacketChain_back_l
  bone2Name=spine_04
  swing1LimitDegrees=45
  swing1Motion=Free
  swing2LimitDegrees=45
  swing2Motion=Free
  twistLimitDegrees=45
  twistMotion=Free
 }
 {
  bone1Name=dyn_jacketChain_back_01_l
  bone2Name=base_jacketChain_back_l
  swing1LimitDegrees=15
  swing2LimitDegrees=40
  twistLimitDegrees=5
  twistMotion=Locked
 }
 {
  bone1Name=dyn_jacketChain_back_02_l
  bone2Name=dyn_jacketChain_back_01_l
  swing1LimitDegrees=15
  swing2LimitDegrees=40
  twistLimitDegrees=5
  twistMotion=Locked
 }
 {
  bone1Name=base_jacketChain_back_r
  bone2Name=spine_04
  swing1LimitDegrees=45
  swing1Motion=Free
  swing2LimitDegrees=45
  swing2Motion=Free
  twistLimitDegrees=45
  twistMotion=Free
 }
 {
  bone1Name=dyn_jacketChain_back_01_r
  bone2Name=base_jacketChain_back_r
  swing1LimitDegrees=25
  swing2LimitDegrees=50
  twistLimitDegrees=5
  twistMotion=Locked
 }
 {
  bone1Name=dyn_jacketChain_back_02_r
  bone2Name=dyn_jacketChain_back_01_r
  swing1LimitDegrees=25
  swing2LimitDegrees=50
  twistLimitDegrees=5
  twistMotion=Locked
 }
 {
  bone1Name=spine_04
  bone2Name=spine_03
  swing1LimitDegrees=45
  swing1Motion=Free
  swing2LimitDegrees=45
  swing2Motion=Free
  twistLimitDegrees=45
  twistMotion=Free
 }
 {
  bone1Name=base_back_chainCollider_L
  bone2Name=spine_04
  swing1LimitDegrees=45
  swing1Motion=Free
  swing2LimitDegrees=45
  swing2Motion=Free
  twistLimitDegrees=45
  twistMotion=Free
 }
 {
  bone1Name=base_back_chainCollider_R
  bone2Name=spine_04
  swing1LimitDegrees=45
  swing1Motion=Free
  swing2LimitDegrees=45
  swing2Motion=Free
  twistLimitDegrees=45
  twistMotion=Free
 }
 {
  bone1Name=neck_01
  bone2Name=spine_05
  swing1LimitDegrees=45
  swing1Motion=Free
  swing2LimitDegrees=45
  swing2Motion=Free
  twistLimitDegrees=45
  twistMotion=Free
 }
 {
  bone1Name=base_necklace
  bone2Name=neck_01
  swing1LimitDegrees=45
  swing1Motion=Free
  swing2LimitDegrees=45
  swing2Motion=Free
  twistLimitDegrees=45
  twistMotion=Free
 }
 {bone1Name=dyn_necklace, bone2Name=base_necklace, swing1LimitDegrees=20, swing2LimitDegrees=30, twistLimitDegrees=45, twistMotion=Locked}
 {
  bone1Name=neck_02
  bone2Name=neck_01
  swing1LimitDegrees=45
  swing1Motion=Free
  swing2LimitDegrees=45
  swing2Motion=Free
  twistLimitDegrees=45
  twistMotion=Free
 }
 {
  bone1Name=spine_05
  bone2Name=spine_04
  swing1LimitDegrees=45
  swing1Motion=Free
  swing2LimitDegrees=45
  swing2Motion=Free
  twistLimitDegrees=45
  twistMotion=Free
 }
]
