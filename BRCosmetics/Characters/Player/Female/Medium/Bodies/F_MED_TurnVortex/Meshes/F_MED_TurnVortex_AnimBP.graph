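# Post-copy AnimBP for the female TurnVortex cosmetic body, adding secondary motion to skirt, laces, pendant and chains.
# The copied parent pose feeds a crouch-aware Control Rig, then a spine_05-space RigidBody pass, cached for two AnimDynamics chains.
# bIsInFrontEnd selects between chains differing only in pendant/chain simulation-space angular alpha, before an additive base_pendant rotation.
# The update event computes front-end, crouch and gravity-override variables, which the shown AnimGraph never reads.
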AnimBlueprint /BRCosmetics/Characters/Player/Female/Medium/Bodies/F_MED_TurnVortex/Meshes/F_MED_TurnVortex_AnimBP
target_skeleton=&/F_MED_TurnVortex_Skeleton
graphs [
 AnimGraph {nodes [
   n1 = Anim Dynamics - Bone: dyn_rounfLaces_01_l(ComponentPose=Local To Component(LocalPose=Use cached pose 'SkirtDynPose'()), Alpha=1, ExternalForce="0,0,0") {
    angularDampingOverride=1
    angularSpringConstant=20000
    bAngularSpring=true
    bOverrideAngularDamping=true
    bUseGravityOverride=true
    gravityScale=0.35
    simulationSpace=RootRelative
    boundBone {boneName=dyn_rounfLaces_01_l}
    componentLinearAccScale=<1 1 1>
    componentLinearVelScale=<1 1 1>
    gravityOverride {z=-980}
    physicsBodyDefinitions [{
      collisionType=CustomSphere
      sphereCollisionRadius=2
      boundBone {boneName=dyn_rounfLaces_01_l}
      boxExtents=<20 30 30>
      constraintSetup {angularLimitsMax=<50 0 50>, angularLimitsMin=<-50 -50 -50>, angularTarget=<1 0 0>}
      localJointOffset=<4 0 0>
     }]
    planarLimits [{drivingBone {boneName=calf_l}, planeTransform {location=<0 1 0>, rotation=<rot 0 0 90>}}]
    relativeSpaceBone {boneName=calf_l}
   }
   n2 = Anim Dynamics - Bone: dyn_rounfLaces_02_l(ComponentPose=@n1, Alpha=1, ExternalForce="0,0,0") {
    angularDampingOverride=1
    angularSpringConstant=20000
    bAngularSpring=true
    bOverrideAngularDamping=true
    bUseGravityOverride=true
    gravityScale=0.35
    simulationSpace=RootRelative
    boundBone {boneName=dyn_rounfLaces_02_l}
    componentLinearAccScale=<1 1 1>
    componentLinearVelScale=<1 1 1>
    gravityOverride {z=-980}
    physicsBodyDefinitions [{
      collisionType=CustomSphere
      sphereCollisionRadius=2
      boundBone {boneName=dyn_rounfLaces_02_l}
      boxExtents=<20 30 30>
      constraintSetup {angularLimitsMax=<50 0 50>, angularLimitsMin=<-50 -50 -50>, angularTarget=<1 0 0>}
      localJointOffset=<4 0 0>
     }]
    planarLimits [{drivingBone {boneName=calf_l}, planeTransform {location=<0 1 0>, rotation=<rot 0 0 90>}}]
    relativeSpaceBone {boneName=calf_l}
    sphericalLimits [{limitRadius=0, drivingBone {boneName=thigh_l}, sphereLocalOffset=<0 0 0>}]
   }
   n3 = Anim Dynamics - Bone: dyn_lace_01_l(ComponentPose=@n2, Alpha=1, ExternalForce="0,0,0") {
    angularDampingOverride=1
    angularSpringConstant=20000
    bAngularSpring=true
    bOverrideAngularDamping=true
    bUseGravityOverride=true
    gravityScale=0.35
    simulationSpace=RootRelative
    boundBone {boneName=dyn_lace_01_l}
    componentLinearAccScale=<1 1 1>
    componentLinearVelScale=<1 1 1>
    gravityOverride {z=-980}
    physicsBodyDefinitions [{
      collisionType=CustomSphere
      sphereCollisionRadius=2
      boundBone {boneName=dyn_lace_01_l}
      boxExtents=<20 30 30>
      constraintSetup {angularLimitsMax=<50 0 50>, angularLimitsMin=<-50 -50 -50>, angularTarget=<1 0 0>}
      localJointOffset=<4 0 0>
     }]
    planarLimits [{drivingBone {boneName=calf_l}, planeTransform {location=<0 1 0>, rotation=<rot 0 0 90>}}]
    relativeSpaceBone {boneName=calf_l}
    sphericalLimits [{limitRadius=0, drivingBone {boneName=thigh_l}, sphereLocalOffset=<0 0 0>}]
   }
   n4 = Anim Dynamics - Bone: dyn_lace_02_l(ComponentPose=@n3, Alpha=1, ExternalForce="0,0,0") {
    angularDampingOverride=1
    angularSpringConstant=20000
    bAngularSpring=true
    bOverrideAngularDamping=true
    bUseGravityOverride=true
    gravityScale=0.35
    simulationSpace=RootRelative
    boundBone {boneName=dyn_lace_02_l}
    componentLinearAccScale=<1 1 1>
    componentLinearVelScale=<1 1 1>
    gravityOverride {z=-980}
    physicsBodyDefinitions [{
      collisionType=CustomSphere
      sphereCollisionRadius=2
      boundBone {boneName=dyn_lace_02_l}
      boxExtents=<20 30 30>
      constraintSetup {angularLimitsMax=<50 0 50>, angularLimitsMin=<-50 -50 -50>, angularTarget=<1 0 0>}
      localJointOffset=<4 0 0>
     }]
    planarLimits [{drivingBone {boneName=calf_l}, planeTransform {location=<0 1 0>, rotation=<rot 0 0 90>}}]
    relativeSpaceBone {boneName=calf_l}
    sphericalLimits [{limitRadius=0, drivingBone {boneName=thigh_l}, sphereLocalOffset=<0 0 0>}]
   }
   n5 = Anim Dynamics - Bone: dyn_rounfLaces_01_r(ComponentPose=@n4, Alpha=1, ExternalForce="0,0,0") {
    angularDampingOverride=1
    angularSpringConstant=20000
    bAngularSpring=true
    bOverrideAngularDamping=true
    bUseGravityOverride=true
    gravityScale=0.35
    simulationSpace=RootRelative
    boundBone {boneName=dyn_rounfLaces_01_r}
    componentLinearAccScale=<1 1 1>
    componentLinearVelScale=<1 1 1>
    gravityOverride {z=-980}
    physicsBodyDefinitions [{
      collisionType=CustomSphere
      sphereCollisionRadius=2
      boundBone {boneName=dyn_rounfLaces_01_r}
      boxExtents=<20 30 30>
      constraintSetup {angularLimitsMax=<50 0 50>, angularLimitsMin=<-50 -50 -50>, angularTarget=<1 0 0>}
      localJointOffset=<4 0 0>
     }]
    relativeSpaceBone {boneName=calf_r}
    sphericalLimits [{limitRadius=0, drivingBone {boneName=None}, sphereLocalOffset=<0 0 0>}]
   }
   n6 = Anim Dynamics - Bone: dyn_rounfLaces_02_r(ComponentPose=@n5, Alpha=1, ExternalForce="0,0,0") {
    angularDampingOverride=1
    angularSpringConstant=20000
    bAngularSpring=true
    bOverrideAngularDamping=true
    bUseGravityOverride=true
    gravityScale=0.35
    simulationSpace=RootRelative
    boundBone {boneName=dyn_rounfLaces_02_r}
    componentLinearAccScale=<1 1 1>
    componentLinearVelScale=<1 1 1>
    gravityOverride {z=-980}
    physicsBodyDefinitions [{
      collisionType=CustomSphere
      sphereCollisionRadius=2
      boundBone {boneName=dyn_rounfLaces_02_r}
      boxExtents=<20 30 30>
      constraintSetup {angularLimitsMax=<50 0 50>, angularLimitsMin=<-50 -50 -50>, angularTarget=<1 0 0>}
      localJointOffset=<4 0 0>
     }]
    relativeSpaceBone {boneName=calf_r}
    sphericalLimits [{limitRadius=0, drivingBone {boneName=None}, sphereLocalOffset=<0 0 0>}]
   }
   n7 = Anim Dynamics - Bone: dyn_race_01_r(ComponentPose=@n6, Alpha=1, ExternalForce="0,0,0") {
    angularDampingOverride=1
    angularSpringConstant=20000
    bAngularSpring=true
    bOverrideAngularDamping=true
    bUseGravityOverride=true
    gravityScale=0.35
    simulationSpace=RootRelative
    boundBone {boneName=dyn_race_01_r}
    componentLinearAccScale=<1 1 1>
    componentLinearVelScale=<1 1 1>
    gravityOverride {z=-980}
    physicsBodyDefinitions [{
      collisionType=CustomSphere
      sphereCollisionRadius=2
      boundBone {boneName=dyn_race_01_r}
      boxExtents=<20 30 30>
      constraintSetup {angularLimitsMax=<50 0 50>, angularLimitsMin=<-50 -50 -50>, angularTarget=<1 0 0>}
      localJointOffset=<4 0 0>
     }]
    relativeSpaceBone {boneName=calf_r}
    sphericalLimits [{limitRadius=0, drivingBone {boneName=None}, sphereLocalOffset=<0 0 0>}]
   }
   n8 = Anim Dynamics - Bone: dyn_race_02_r(ComponentPose=@n7, Alpha=1, ExternalForce="0,0,0") {
    angularDampingOverride=1
    angularSpringConstant=20000
    bAngularSpring=true
    bOverrideAngularDamping=true
    bUseGravityOverride=true
    gravityScale=0.35
    simulationSpace=RootRelative
    boundBone {boneName=dyn_race_02_r}
    componentLinearAccScale=<1 1 1>
    componentLinearVelScale=<1 1 1>
    gravityOverride {z=-980}
    physicsBodyDefinitions [{
      collisionType=CustomSphere
      sphereCollisionRadius=2
      boundBone {boneName=dyn_race_02_r}
      boxExtents=<20 30 30>
      constraintSetup {angularLimitsMax=<50 0 50>, angularLimitsMin=<-50 -50 -50>, angularTarget=<1 0 0>}
      localJointOffset=<4 0 0>
     }]
    relativeSpaceBone {boneName=calf_r}
    sphericalLimits [{limitRadius=0, drivingBone {boneName=None}, sphereLocalOffset=<0 0 0>}]
   }
   n9 = Anim Dynamics - Bone: dyn_pendant(ComponentPose=@n8, Alpha=1, ExternalForce="0,0,0") {
    angularDampingOverride=1
    angularSpringConstant=5000
    bAngularSpring=true
    bOverrideAngularDamping=true
    bUseGravityOverride=true
    bUsePlanarLimit=false
    bUseSphericalLimits=true
    gravityScale=0.35
    simulationSpace=RootRelative
    boundBone {boneName=dyn_pendant}
    componentAppliedLinearAccClamp=<3000 3000 3000>
    componentLinearAccScale=<1 1 1>
    componentLinearVelScale=<1 1 1>
    gravityOverride {z=-980}
    physicsBodyDefinitions [{
      boundBone {boneName=dyn_pendant}
      boxExtents=<20 30 30>
      constraintSetup {angularLimitsMax=<10 0 60>, angularLimitsMin=<-10 -150 -100>, angularTarget=<1 0 0>}
      localJointOffset=<5 0 0>
     }]
    simSpaceSettings {maxAngularAcceleration=1000, maxAngularVelocity=1000, simSpaceAngularAlpha=1, externalAngularVelocity=<1 1 1>}
    sphericalLimits [{limitRadius=9, drivingBone {boneName=pelvis}, sphereLocalOffset=<0 0 -2>}]
   }
   n10 = Anim Dynamics - Bone: dyn_pendant(ComponentPose=@n9, Alpha=1, ExternalForce="0,0,0") {
    angularDampingOverride=1
    angularSpringConstant=2000
    bAngularSpring=true
    bOverrideAngularDamping=true
    bUseGravityOverride=true
    bUsePlanarLimit=false
    bUseSphericalLimits=true
    gravityScale=0.35
    simulationSpace=RootRelative
    boundBone {boneName=dyn_pendant}
    componentAppliedLinearAccClamp=<3000 3000 3000>
    componentLinearAccScale=<1 1 1>
    componentLinearVelScale=<1 1 1>
    gravityOverride {z=-980}
    physicsBodyDefinitions [{
      boundBone {boneName=dyn_pendant}
      boxExtents=<20 30 30>
      constraintSetup {angularLimitsMax=<7 0 20>, angularLimitsMin=<-7 -50 -20>, angularTarget=<1 0 0>}
      localJointOffset=<4 0 0>
     }]
    simSpaceSettings {maxAngularAcceleration=1000, maxAngularVelocity=1000, simSpaceAngularAlpha=1, externalAngularVelocity=<1 1 1>}
    sphericalLimits [{limitRadius=9.5, drivingBone {boneName=thigh_l}, sphereLocalOffset=<-2 0 0>}]
   }
   n11 = Anim Dynamics - Bone: dyn_frontChains_01(ComponentPose=@n10, Alpha=1, ExternalForce="0,0,0") {
    angularDampingOverride=1
    angularSpringConstant=10000
    bAngularSpring=true
    bOverrideAngularDamping=true
    bUseGravityOverride=true
    gravityScale=0.35
    simulationSpace=RootRelative
    boundBone {boneName=dyn_frontChains_01}
    componentAppliedLinearAccClamp=<3000 3000 3000>
    componentLinearAccScale=<1 1 1>
    componentLinearVelScale=<1 1 1>
    gravityOverride {z=-980}
    physicsBodyDefinitions [{
      boundBone {boneName=dyn_frontChains_01}
      boxExtents=<20 30 30>
      constraintSetup {angularLimitsMax=<3 0 3>, angularLimitsMin=<-3 0 0>, angularTarget=<1 0 0>}
      localJointOffset=<2 0 0>
     }]
    simSpaceSettings {maxAngularAcceleration=1000, maxAngularVelocity=1000, simSpaceAngularAlpha=0.5}
   }
   n12 = Anim Dynamics - Bone: dyn_frontChains_02(ComponentPose=@n11, Alpha=1, ExternalForce="0,0,0") {
    angularDampingOverride=1
    angularSpringConstant=10000
    bAngularSpring=true
    bOverrideAngularDamping=true
    bUseGravityOverride=true
    gravityScale=0.35
    simulationSpace=RootRelative
    boundBone {boneName=dyn_frontChains_02}
    componentAppliedLinearAccClamp=<3000 3000 3000>
    componentLinearAccScale=<1 1 1>
    componentLinearVelScale=<1 1 1>
    gravityOverride {z=-980}
    physicsBodyDefinitions [{
      boundBone {boneName=dyn_frontChains_02}
      boxExtents=<20 30 30>
      constraintSetup {angularLimitsMax=<3 0 6>, angularLimitsMin=<-3 0 0>, angularTarget=<1 0 0>}
      localJointOffset=<2 0 0>
     }]
    simSpaceSettings {maxAngularAcceleration=1000, maxAngularVelocity=1000, simSpaceAngularAlpha=0.5}
   }
   n13 = Anim Dynamics - Bone: dyn_frontChains_03(ComponentPose=@n12, Alpha=1, ExternalForce="0,0,0") {
    angularDampingOverride=1
    angularSpringConstant=10000
    bAngularSpring=true
    bOverrideAngularDamping=true
    bUseGravityOverride=true
    gravityScale=0.35
    simulationSpace=RootRelative
    boundBone {boneName=dyn_frontChains_03}
    componentAppliedLinearAccClamp=<3000 3000 3000>
    componentLinearAccScale=<1 1 1>
    componentLinearVelScale=<1 1 1>
    gravityOverride {z=-980}
    physicsBodyDefinitions [{
      boundBone {boneName=dyn_frontChains_03}
      boxExtents=<20 30 30>
      constraintSetup {angularLimitsMax=<3 0 12>, angularLimitsMin=<-3 0 0>, angularTarget=<1 0 0>}
      localJointOffset=<2 0 0>
     }]
    simSpaceSettings {maxAngularAcceleration=1000, maxAngularVelocity=1000, simSpaceAngularAlpha=0.5}
   }
   n14 = Anim Dynamics - Bone: dyn_backChains_01(ComponentPose=@n13, Alpha=1, ExternalForce="0,0,0") {
    angularDampingOverride=1
    angularSpringConstant=10000
    bAngularSpring=true
    bOverrideAngularDamping=true
    bUseGravityOverride=true
    gravityScale=0.35
    simulationSpace=RootRelative
    boundBone {boneName=dyn_backChains_01}
    componentAppliedLinearAccClamp=<3000 3000 3000>
    componentLinearAccScale=<1 1 1>
    componentLinearVelScale=<1 1 1>
    gravityOverride {z=-980}
    physicsBodyDefinitions [{
      boundBone {boneName=dyn_backChains_01}
      boxExtents=<20 30 30>
      constraintSetup {angularLimitsMax=<3 0 0>, angularLimitsMin=<-3 0 -3>, angularTarget=<1 0 0>}
      localJointOffset=<2 0 0>
     }]
    simSpaceSettings {maxAngularAcceleration=1000, maxAngularVelocity=1000, simSpaceAngularAlpha=1}
   }
   n15 = Anim Dynamics - Bone: dyn_backChains_02(ComponentPose=@n14, Alpha=1, ExternalForce="0,0,0") {
    angularDampingOverride=1
    angularSpringConstant=10000
    bAngularSpring=true
    bOverrideAngularDamping=true
    bUseGravityOverride=true
    gravityScale=0.35
    simulationSpace=RootRelative
    boundBone {boneName=dyn_backChains_02}
    componentAppliedLinearAccClamp=<3000 3000 3000>
    componentLinearAccScale=<1 1 1>
    componentLinearVelScale=<1 1 1>
    gravityOverride {z=-980}
    physicsBodyDefinitions [{
      boundBone {boneName=dyn_backChains_02}
      boxExtents=<20 30 30>
      constraintSetup {angularLimitsMax=<3 0 0>, angularLimitsMin=<-3 0 -12>, angularTarget=<1 0 0>}
      localJointOffset=<2 0 0>
     }]
    simSpaceSettings {maxAngularAcceleration=1000, maxAngularVelocity=1000, simSpaceAngularAlpha=1}
   }
   n16 = Anim Dynamics - Bone: dyn_backChains_03(ComponentPose=@n15, Alpha=1, ExternalForce="0,0,0") {
    angularDampingOverride=1
    angularSpringConstant=10000
    bAngularSpring=true
    bOverrideAngularDamping=true
    bUseGravityOverride=true
    gravityScale=0.35
    simulationSpace=RootRelative
    boundBone {boneName=dyn_backChains_03}
    componentAppliedLinearAccClamp=<3000 3000 3000>
    componentLinearAccScale=<1 1 1>
    componentLinearVelScale=<1 1 1>
    gravityOverride {z=-980}
    physicsBodyDefinitions [{
      boundBone {boneName=dyn_backChains_03}
      boxExtents=<20 30 30>
      constraintSetup {angularLimitsMax=<3 0 0>, angularLimitsMin=<-3 0 -12>, angularTarget=<1 0 0>}
      localJointOffset=<2 0 0>
     }]
    simSpaceSettings {maxAngularAcceleration=1000, maxAngularVelocity=1000, simSpaceAngularAlpha=1}
   }
   n17 = Component To Local(ComponentPose=@n16)
   n18 = Anim Dynamics - Bone: dyn_rounfLaces_01_l(ComponentPose=Local To Component(LocalPose=Use cached pose 'SkirtDynPose'()), Alpha=1, ExternalForce="0,0,0") {
    angularDampingOverride=1
    angularSpringConstant=20000
    bAngularSpring=true
    bOverrideAngularDamping=true
    bUseGravityOverride=true
    gravityScale=0.35
    simulationSpace=RootRelative
    boundBone {boneName=dyn_rounfLaces_01_l}
    componentLinearAccScale=<1 1 1>
    componentLinearVelScale=<1 1 1>
    gravityOverride {z=-980}
    physicsBodyDefinitions [{
      collisionType=CustomSphere
      sphereCollisionRadius=2
      boundBone {boneName=dyn_rounfLaces_01_l}
      boxExtents=<20 30 30>
      constraintSetup {angularLimitsMax=<50 0 50>, angularLimitsMin=<-50 -50 -50>, angularTarget=<1 0 0>}
      localJointOffset=<4 0 0>
     }]
    planarLimits [{drivingBone {boneName=calf_l}, planeTransform {location=<0 1 0>, rotation=<rot 0 0 90>}}]
    relativeSpaceBone {boneName=calf_l}
   }
   n19 = Anim Dynamics - Bone: dyn_rounfLaces_02_l(ComponentPose=@n18, Alpha=1, ExternalForce="0,0,0") {
    angularDampingOverride=1
    angularSpringConstant=20000
    bAngularSpring=true
    bOverrideAngularDamping=true
    bUseGravityOverride=true
    gravityScale=0.35
    simulationSpace=RootRelative
    boundBone {boneName=dyn_rounfLaces_02_l}
    componentLinearAccScale=<1 1 1>
    componentLinearVelScale=<1 1 1>
    gravityOverride {z=-980}
    physicsBodyDefinitions [{
      collisionType=CustomSphere
      sphereCollisionRadius=2
      boundBone {boneName=dyn_rounfLaces_02_l}
      boxExtents=<20 30 30>
      constraintSetup {angularLimitsMax=<50 0 50>, angularLimitsMin=<-50 -50 -50>, angularTarget=<1 0 0>}
      localJointOffset=<4 0 0>
     }]
    planarLimits [{drivingBone {boneName=calf_l}, planeTransform {location=<0 1 0>, rotation=<rot 0 0 90>}}]
    relativeSpaceBone {boneName=calf_l}
    sphericalLimits [{limitRadius=0, drivingBone {boneName=thigh_l}, sphereLocalOffset=<0 0 0>}]
   }
   n20 = Anim Dynamics - Bone: dyn_lace_01_l(ComponentPose=@n19, Alpha=1, ExternalForce="0,0,0") {
    angularDampingOverride=1
    angularSpringConstant=20000
    bAngularSpring=true
    bOverrideAngularDamping=true
    bUseGravityOverride=true
    gravityScale=0.35
    simulationSpace=RootRelative
    boundBone {boneName=dyn_lace_01_l}
    componentLinearAccScale=<1 1 1>
    componentLinearVelScale=<1 1 1>
    gravityOverride {z=-980}
    physicsBodyDefinitions [{
      collisionType=CustomSphere
      sphereCollisionRadius=2
      boundBone {boneName=dyn_lace_01_l}
      boxExtents=<20 30 30>
      constraintSetup {angularLimitsMax=<50 0 50>, angularLimitsMin=<-50 -50 -50>, angularTarget=<1 0 0>}
      localJointOffset=<4 0 0>
     }]
    planarLimits [{drivingBone {boneName=calf_l}, planeTransform {location=<0 1 0>, rotation=<rot 0 0 90>}}]
    relativeSpaceBone {boneName=calf_l}
    sphericalLimits [{limitRadius=0, drivingBone {boneName=thigh_l}, sphereLocalOffset=<0 0 0>}]
   }
   n21 = Anim Dynamics - Bone: dyn_lace_02_l(ComponentPose=@n20, Alpha=1, ExternalForce="0,0,0") {
    angularDampingOverride=1
    angularSpringConstant=20000
    bAngularSpring=true
    bOverrideAngularDamping=true
    bUseGravityOverride=true
    gravityScale=0.35
    simulationSpace=RootRelative
    boundBone {boneName=dyn_lace_02_l}
    componentLinearAccScale=<1 1 1>
    componentLinearVelScale=<1 1 1>
    gravityOverride {z=-980}
    physicsBodyDefinitions [{
      collisionType=CustomSphere
      sphereCollisionRadius=2
      boundBone {boneName=dyn_lace_02_l}
      boxExtents=<20 30 30>
      constraintSetup {angularLimitsMax=<50 0 50>, angularLimitsMin=<-50 -50 -50>, angularTarget=<1 0 0>}
      localJointOffset=<4 0 0>
     }]
    planarLimits [{drivingBone {boneName=calf_l}, planeTransform {location=<0 1 0>, rotation=<rot 0 0 90>}}]
    relativeSpaceBone {boneName=calf_l}
    sphericalLimits [{limitRadius=0, drivingBone {boneName=thigh_l}, sphereLocalOffset=<0 0 0>}]
   }
   n22 = Anim Dynamics - Bone: dyn_rounfLaces_01_r(ComponentPose=@n21, Alpha=1, ExternalForce="0,0,0") {
    angularDampingOverride=1
    angularSpringConstant=20000
    bAngularSpring=true
    bOverrideAngularDamping=true
    bUseGravityOverride=true
    gravityScale=0.35
    simulationSpace=RootRelative
    boundBone {boneName=dyn_rounfLaces_01_r}
    componentLinearAccScale=<1 1 1>
    componentLinearVelScale=<1 1 1>
    gravityOverride {z=-980}
    physicsBodyDefinitions [{
      collisionType=CustomSphere
      sphereCollisionRadius=2
      boundBone {boneName=dyn_rounfLaces_01_r}
      boxExtents=<20 30 30>
      constraintSetup {angularLimitsMax=<50 0 50>, angularLimitsMin=<-50 -50 -50>, angularTarget=<1 0 0>}
      localJointOffset=<4 0 0>
     }]
    relativeSpaceBone {boneName=calf_r}
    sphericalLimits [{limitRadius=0, drivingBone {boneName=None}, sphereLocalOffset=<0 0 0>}]
   }
   n23 = Anim Dynamics - Bone: dyn_rounfLaces_02_r(ComponentPose=@n22, Alpha=1, ExternalForce="0,0,0") {
    angularDampingOverride=1
    angularSpringConstant=20000
    bAngularSpring=true
    bOverrideAngularDamping=true
    bUseGravityOverride=true
    gravityScale=0.35
    simulationSpace=RootRelative
    boundBone {boneName=dyn_rounfLaces_02_r}
    componentLinearAccScale=<1 1 1>
    componentLinearVelScale=<1 1 1>
    gravityOverride {z=-980}
    physicsBodyDefinitions [{
      collisionType=CustomSphere
      sphereCollisionRadius=2
      boundBone {boneName=dyn_rounfLaces_02_r}
      boxExtents=<20 30 30>
      constraintSetup {angularLimitsMax=<50 0 50>, angularLimitsMin=<-50 -50 -50>, angularTarget=<1 0 0>}
      localJointOffset=<4 0 0>
     }]
    relativeSpaceBone {boneName=calf_r}
    sphericalLimits [{limitRadius=0, drivingBone {boneName=None}, sphereLocalOffset=<0 0 0>}]
   }
   n24 = Anim Dynamics - Bone: dyn_race_01_r(ComponentPose=@n23, Alpha=1, ExternalForce="0,0,0") {
    angularDampingOverride=1
    angularSpringConstant=20000
    bAngularSpring=true
    bOverrideAngularDamping=true
    bUseGravityOverride=true
    gravityScale=0.35
    simulationSpace=RootRelative
    boundBone {boneName=dyn_race_01_r}
    componentLinearAccScale=<1 1 1>
    componentLinearVelScale=<1 1 1>
    gravityOverride {z=-980}
    physicsBodyDefinitions [{
      collisionType=CustomSphere
      sphereCollisionRadius=2
      boundBone {boneName=dyn_race_01_r}
      boxExtents=<20 30 30>
      constraintSetup {angularLimitsMax=<50 0 50>, angularLimitsMin=<-50 -50 -50>, angularTarget=<1 0 0>}
      localJointOffset=<4 0 0>
     }]
    relativeSpaceBone {boneName=calf_r}
    sphericalLimits [{limitRadius=0, drivingBone {boneName=None}, sphereLocalOffset=<0 0 0>}]
   }
   n25 = Anim Dynamics - Bone: dyn_race_02_r(ComponentPose=@n24, Alpha=1, ExternalForce="0,0,0") {
    angularDampingOverride=1
    angularSpringConstant=20000
    bAngularSpring=true
    bOverrideAngularDamping=true
    bUseGravityOverride=true
    gravityScale=0.35
    simulationSpace=RootRelative
    boundBone {boneName=dyn_race_02_r}
    componentLinearAccScale=<1 1 1>
    componentLinearVelScale=<1 1 1>
    gravityOverride {z=-980}
    physicsBodyDefinitions [{
      collisionType=CustomSphere
      sphereCollisionRadius=2
      boundBone {boneName=dyn_race_02_r}
      boxExtents=<20 30 30>
      constraintSetup {angularLimitsMax=<50 0 50>, angularLimitsMin=<-50 -50 -50>, angularTarget=<1 0 0>}
      localJointOffset=<4 0 0>
     }]
    relativeSpaceBone {boneName=calf_r}
    sphericalLimits [{limitRadius=0, drivingBone {boneName=None}, sphereLocalOffset=<0 0 0>}]
   }
   n26 = Anim Dynamics - Bone: dyn_pendant(ComponentPose=@n25, Alpha=1, ExternalForce="0,0,0") {
    angularDampingOverride=1
    angularSpringConstant=5000
    bAngularSpring=true
    bOverrideAngularDamping=true
    bUseGravityOverride=true
    bUsePlanarLimit=false
    bUseSphericalLimits=true
    gravityScale=0.35
    simulationSpace=RootRelative
    boundBone {boneName=dyn_pendant}
    componentAppliedLinearAccClamp=<3000 3000 3000>
    componentLinearAccScale=<1 1 1>
    componentLinearVelScale=<1 1 1>
    gravityOverride {z=-980}
    physicsBodyDefinitions [{
      boundBone {boneName=dyn_pendant}
      boxExtents=<20 30 30>
      constraintSetup {angularLimitsMax=<10 0 60>, angularLimitsMin=<-10 -150 -100>, angularTarget=<1 0 0>}
      localJointOffset=<5 0 0>
     }]
    simSpaceSettings {maxAngularAcceleration=1000, maxAngularVelocity=1000, simSpaceAngularAlpha=0.4, externalAngularVelocity=<1 1 1>}
    sphericalLimits [{limitRadius=9, drivingBone {boneName=pelvis}, sphereLocalOffset=<0 0 -2>}]
   }
   n27 = Anim Dynamics - Bone: dyn_pendant(ComponentPose=@n26, Alpha=1, ExternalForce="0,0,0") {
    angularDampingOverride=1
    angularSpringConstant=2000
    bAngularSpring=true
    bOverrideAngularDamping=true
    bUseGravityOverride=true
    bUsePlanarLimit=false
    bUseSphericalLimits=true
    gravityScale=0.35
    simulationSpace=RootRelative
    boundBone {boneName=dyn_pendant}
    componentAppliedLinearAccClamp=<3000 3000 3000>
    componentLinearAccScale=<1 1 1>
    componentLinearVelScale=<1 1 1>
    gravityOverride {z=-980}
    physicsBodyDefinitions [{
      boundBone {boneName=dyn_pendant}
      boxExtents=<20 30 30>
      constraintSetup {angularLimitsMax=<7 0 20>, angularLimitsMin=<-7 -50 -20>, angularTarget=<1 0 0>}
      localJointOffset=<4 0 0>
     }]
    simSpaceSettings {maxAngularAcceleration=1000, maxAngularVelocity=1000, simSpaceAngularAlpha=0.4, externalAngularVelocity=<1 1 1>}
    sphericalLimits [{limitRadius=9.5, drivingBone {boneName=thigh_l}, sphereLocalOffset=<-2 0 0>}]
   }
   n28 = Anim Dynamics - Bone: dyn_frontChains_01(ComponentPose=@n27, Alpha=1, ExternalForce="0,0,0") {
    angularDampingOverride=1
    angularSpringConstant=10000
    bAngularSpring=true
    bOverrideAngularDamping=true
    bUseGravityOverride=true
    gravityScale=0.35
    simulationSpace=RootRelative
    boundBone {boneName=dyn_frontChains_01}
    componentAppliedLinearAccClamp=<3000 3000 3000>
    componentLinearAccScale=<1 1 1>
    componentLinearVelScale=<1 1 1>
    gravityOverride {z=-980}
    physicsBodyDefinitions [{
      boundBone {boneName=dyn_frontChains_01}
      boxExtents=<20 30 30>
      constraintSetup {angularLimitsMax=<3 0 3>, angularLimitsMin=<-3 0 0>, angularTarget=<1 0 0>}
      localJointOffset=<2 0 0>
     }]
    simSpaceSettings {maxAngularAcceleration=1000, maxAngularVelocity=1000, simSpaceAngularAlpha=0.2}
   }
   n29 = Anim Dynamics - Bone: dyn_frontChains_02(ComponentPose=@n28, Alpha=1, ExternalForce="0,0,0") {
    angularDampingOverride=1
    angularSpringConstant=10000
    bAngularSpring=true
    bOverrideAngularDamping=true
    bUseGravityOverride=true
    gravityScale=0.35
    simulationSpace=RootRelative
    boundBone {boneName=dyn_frontChains_02}
    componentAppliedLinearAccClamp=<3000 3000 3000>
    componentLinearAccScale=<1 1 1>
    componentLinearVelScale=<1 1 1>
    gravityOverride {z=-980}
    physicsBodyDefinitions [{
      boundBone {boneName=dyn_frontChains_02}
      boxExtents=<20 30 30>
      constraintSetup {angularLimitsMax=<3 0 6>, angularLimitsMin=<-3 0 0>, angularTarget=<1 0 0>}
      localJointOffset=<2 0 0>
     }]
    simSpaceSettings {maxAngularAcceleration=1000, maxAngularVelocity=1000, simSpaceAngularAlpha=0.2}
   }
   n30 = Anim Dynamics - Bone: dyn_frontChains_03(ComponentPose=@n29, Alpha=1, ExternalForce="0,0,0") {
    angularDampingOverride=1
    angularSpringConstant=10000
    bAngularSpring=true
    bOverrideAngularDamping=true
    bUseGravityOverride=true
    gravityScale=0.35
    simulationSpace=RootRelative
    boundBone {boneName=dyn_frontChains_03}
    componentAppliedLinearAccClamp=<3000 3000 3000>
    componentLinearAccScale=<1 1 1>
    componentLinearVelScale=<1 1 1>
    gravityOverride {z=-980}
    physicsBodyDefinitions [{
      boundBone {boneName=dyn_frontChains_03}
      boxExtents=<20 30 30>
      constraintSetup {angularLimitsMax=<3 0 12>, angularLimitsMin=<-3 0 0>, angularTarget=<1 0 0>}
      localJointOffset=<2 0 0>
     }]
    simSpaceSettings {maxAngularAcceleration=1000, maxAngularVelocity=1000, simSpaceAngularAlpha=0.2}
   }
   n31 = Anim Dynamics - Bone: dyn_backChains_01(ComponentPose=@n30, Alpha=1, ExternalForce="0,0,0") {
    angularDampingOverride=1
    angularSpringConstant=10000
    bAngularSpring=true
    bOverrideAngularDamping=true
    bUseGravityOverride=true
    gravityScale=0.35
    simulationSpace=RootRelative
    boundBone {boneName=dyn_backChains_01}
    componentAppliedLinearAccClamp=<3000 3000 3000>
    componentLinearAccScale=<1 1 1>
    componentLinearVelScale=<1 1 1>
    gravityOverride {z=-980}
    physicsBodyDefinitions [{
      boundBone {boneName=dyn_backChains_01}
      boxExtents=<20 30 30>
      constraintSetup {angularLimitsMax=<3 0 0>, angularLimitsMin=<-3 0 -3>, angularTarget=<1 0 0>}
      localJointOffset=<2 0 0>
     }]
    simSpaceSettings {maxAngularAcceleration=1000, maxAngularVelocity=1000, simSpaceAngularAlpha=0.4}
   }
   n32 = Anim Dynamics - Bone: dyn_backChains_02(ComponentPose=@n31, Alpha=1, ExternalForce="0,0,0") {
    angularDampingOverride=1
    angularSpringConstant=10000
    bAngularSpring=true
    bOverrideAngularDamping=true
    bUseGravityOverride=true
    gravityScale=0.35
    simulationSpace=RootRelative
    boundBone {boneName=dyn_backChains_02}
    componentAppliedLinearAccClamp=<3000 3000 3000>
    componentLinearAccScale=<1 1 1>
    componentLinearVelScale=<1 1 1>
    gravityOverride {z=-980}
    physicsBodyDefinitions [{
      boundBone {boneName=dyn_backChains_02}
      boxExtents=<20 30 30>
      constraintSetup {angularLimitsMax=<3 0 0>, angularLimitsMin=<-3 0 -12>, angularTarget=<1 0 0>}
      localJointOffset=<2 0 0>
     }]
    simSpaceSettings {maxAngularAcceleration=1000, maxAngularVelocity=1000, simSpaceAngularAlpha=0.4}
   }
   n33 = Anim Dynamics - Bone: dyn_backChains_03(ComponentPose=@n32, Alpha=1, ExternalForce="0,0,0") {
    angularDampingOverride=1
    angularSpringConstant=10000
    bAngularSpring=true
    bOverrideAngularDamping=true
    bUseGravityOverride=true
    gravityScale=0.35
    simulationSpace=RootRelative
    boundBone {boneName=dyn_backChains_03}
    componentAppliedLinearAccClamp=<3000 3000 3000>
    componentLinearAccScale=<1 1 1>
    componentLinearVelScale=<1 1 1>
    gravityOverride {z=-980}
    physicsBodyDefinitions [{
      boundBone {boneName=dyn_backChains_03}
      boxExtents=<20 30 30>
      constraintSetup {angularLimitsMax=<3 0 0>, angularLimitsMin=<-3 0 -12>, angularTarget=<1 0 0>}
      localJointOffset=<2 0 0>
     }]
    simSpaceSettings {maxAngularAcceleration=1000, maxAngularVelocity=1000, simSpaceAngularAlpha=0.4}
   }
   n34 = Component To Local(ComponentPose=@n33)
   n35 = Get bIsInFrontEnd()
   n36 = "Transform (Modify) Bone - Bone: base_pendant"(
    ComponentPose=Local To Component(LocalPose=Blend Poses by bool(BlendPose_0=@n34, BlendPose_1=@n17, bActiveValue=@n35))
    Alpha=1
    Translation="0,0,0"
    Rotation="-2.277,-16.094,11.619"
    Scale="1,1,1"
   ) {
    rotationMode=BMM_Additive
    rotationSpace=BCS_BoneSpace
    scaleMode=BMM_Additive
    scaleSpace=BCS_BoneSpace
    translationMode=BMM_Additive
    translationSpace=BCS_BoneSpace
    alphaScaleBiasClamp {bInterpResult=true}
    boneToModify {boneName=base_pendant}
    rotation=<rot -2.277 -16.094 11.619>
   }
   Output Pose(Result=Component To Local(ComponentPose=@n36)) {layerGroup=None}
   n37 = RigidBody!(
    ComponentPose=Local To Component(LocalPose=Use cached pose 'CR_Start'())
    bAlphaBoolEnabled:bool
    OverrideWorldGravity:Vector="0,0,-980"
    ExternalForce
    SimSpaceSettings:Sim Space Settings Structure=(WorldAlpha=0,VelocityScaleZ=1,DampingAlpha=1,MaxLinearVelocity=10000,MaxAngularVelocity=10000,MaxLinearAcceleration=10000,MaxAngularAcceleration=10000,ExternalLinearDragV=(X=0,Y=0,Z=0),ExternalLinearVelocity=(X=0,Y=0,Z=0),ExternalAngularVelocity=(X=0,Y=0,Z=0))
   ) -> (Pose) {
    alphaInputType=Bool
    bClampLinearTranslationLimitToRefPose=true
    bOverrideWorldGravity=true
    overridePhysicsAsset=&/F_MED_TurnVortex_Physics
    simulationSpace=BaseBoneSpace
    alphaBoolBlend {blendOption=Linear}
    baseBoneRef {boneName=spine_05}
    componentAppliedLinearAccClamp=<0 0 0>
    overrideWorldGravity {z=-980}
   }
   Save cached pose 'SkirtDynPose'(Pose=Component To Local(ComponentPose=@n37)) {cacheName=SkirtDynPose}
   n38 = Control Rig!(Source=Copy Pose From Mesh() {bUseAttachedParent=true}, Alpha, isCrouching:bool=Get bIsCrouching()) -> (Pose) {controlRigAssetReference {blueprintRigClass=&/F_MED_TurnVortex_CtrlRig.F_MED_TurnVortex_CtrlRig_C}}
   Save cached pose 'CR_Start'(Pose=@n38) {cacheName=CR_Start}
  ]}
 EventGraph {nodes [
   Event BlueprintUpdateAnimation()
   > n1 = Sequence!(execute) -> (then_0, then_1, then_2:Exec)
   n2 = Branch(execute=@n1.then_1, Condition=Get bIsInFrontEnd())
   > Set IdleisPlaying Float(IdleisPlaying Float:double=GetCurveValue(self=GetOwnerFrontEndAnimInstance(), CurveName=Frontend_Idle_Bypass)) -> (Output_Get:double)
   > Set IdleisPlayingArmDown float(IdleisPlayingArmDown float:double=float * float(A=GetCurveValue(CurveName=spr_upperarm_l_dn_alpha), B=Get IdleisPlaying Float() -> (IdleisPlaying Float:double)) {is=title K2Node_PromotableOperator}) -> (Output_Get:double)
   Set IdleisPlaying Float(execute=@n2.else, IdleisPlaying Float=0)
   Set Alpha_Crouching(execute=@n1.then_2, Alpha_Crouching:double="To Float (Boolean)"(InBool=Get bIsCrouching())) -> (Output_Get:double)
   n3 = Get bIsCrouching()
   > Set Alpha_Standing(Alpha_Standing:double="To Float (Boolean)"(InBool=NOT Boolean(A=@n3))) -> (Output_Get:double)
   n4 = NOT Boolean(A=Get bIsMoving2D())
   n5 = Get bIsCrouching()
   > Set Alpha_CrouchingIdle(Alpha_CrouchingIdle:double="To Float (Boolean)"(InBool=AND Boolean(A=@n4, B=@n5))) -> (Output_Get:double)
   > Set NotFrontEnd(NotFrontEnd:bool=NOT Boolean(A=Get bIsInFrontEnd())) -> (Output_Get:bool)
   n6 = "Gravity Override State Function (Parameter Asset) (Nativized)"(
    execute=@n1.then_0
    CharacterPartAnimInstance=Self-Reference()
    GameTimeInSeconds=GetGameTimeInSeconds()
    Parameters=/BRCosmetics/Characters/Player/Female/Medium/Bodies/F_MED_TurnVortex/Meshes/Parts/F_MED_TurnVortex_FaceAcc_Parameters
   )
   > Set Gravity Override Cape(Gravity Override Cape:Vector=@n6) -> (Output_Get:Vector)
   > Set RBAN Settings Out_0(RBAN Settings Out_0:Fort Rigid Body Anim Node Parameters Structure=@n6.OutRigidBodyAnimNodeParameters) -> (Output_Get:Fort Rigid Body Anim Node Parameters Structure)
  ]}
]
variables [
 Alpha_Crouching {type=number}
 Alpha_CrouchingIdle {type=number}
 Alpha_Standing {type=number}
 Gravity Override Cape {type=Vector}
 IdleisPlaying Float {type=number}
 IdleisPlayingArmDown float {type=number}
 NotFrontEnd {type=boolean}
 RBAN Settings Out_0 {type=FortRigidBodyAnimNodeParameters}
]
