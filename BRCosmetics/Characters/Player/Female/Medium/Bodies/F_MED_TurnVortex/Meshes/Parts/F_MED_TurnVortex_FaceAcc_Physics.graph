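# Physics asset for the TurnVortex female medium face-accessory mesh, covering a long back-hair chain and two dangling earrings.
# Capsules on head, neck, torso, limbs and a very large root body form a body proxy linked by free-swing constraints.
# Hair hangs from the head as a long chain forking near the tip, limited mostly to one swing axis with minimal twist.
# Earrings swing within limits from head-attached base bodies, with twist locked.
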PhysicsAsset /BRCosmetics/Characters/Player/Female/Medium/Bodies/F_MED_TurnVortex/Meshes/Parts/F_MED_TurnVortex_FaceAcc_Physics
bodies {
 base_earring_l [{shapeType=Capsule, length=0.505, radius=0.505, shapeName=base_earring_l_capsule, rotation=<rot 0 -90 0>}]
 base_earring_r [{shapeType=Capsule, length=0.505, radius=0.505, shapeName=base_earring_r_capsule, rotation=<rot 0 -90 0>}]
 calf_l [{shapeType=Capsule, length=51, radius=6.5, shapeName=calf_l_capsule, center=<-19.3244 0.0001 -0.8183>, rotation=<rot 0 -90.0002 87.5752>}]
 calf_r [{shapeType=Capsule, length=51, radius=6.5, shapeName=calf_r_capsule, center=<18.7878 0 0.7956>, rotation=<rot 0 -90.0002 87.5752>}]
 clavicle_l [{
   shapeType=Capsule
   length=5.5517
   radius=7.3175
   shapeName=clavicle_l_capsule
   center=<6.9223 0.5494 0.217>
   rotation=<rot -0.0001 -90 -91.7954>
  }]
 clavicle_r [{
   shapeType=Capsule
   length=5.6878
   radius=7.3175
   shapeName=clavicle_r_capsule
   center=<-6.0019 -0.3208 -0.1063>
   rotation=<rot -0.0001 -90 88.9855>
  }]
 dyn_Hair_Back_01 [{
   shapeType=Capsule
   length=5.6007
   radius=4.2147
   shapeName=dyn_Hair_Back_01_capsule
   center=<0.4811 -3.8564 -0.069>
   rotation=<rot 78.3227 -79.4314 3.9317>
  }]
 dyn_Hair_Back_02 [{
   shapeType=Capsule
   length=7.1379
   radius=5.1048
   shapeName=dyn_Hair_Back_02_capsule
   center=<3.0999 -3.4172 0.2639>
   rotation=<rot -77.7509 -177.8079 101.4956>
  }]
 dyn_Hair_Back_03 [{
   shapeType=Capsule
   length=5.432
   radius=6.7839
   shapeName=dyn_Hair_Back_03_capsule
   center=<2.6983 -1.4203 0.0101>
   rotation=<rot 0 -0.0003 90.4093>
  }]
 dyn_Hair_Back_04 [{
   shapeType=Capsule
   length=6.7015
   radius=6.3856
   shapeName=dyn_Hair_Back_04_capsule
   center=<2.6983 -1.4203 0.0101>
   rotation=<rot 0 -0.0003 90.4093>
  }]
 dyn_Hair_Back_05 [{
   shapeType=Capsule
   length=3.0606
   radius=5.7116
   shapeName=dyn_Hair_Back_06_capsule
   center=<2.6983 -3.7731 0.027>
   rotation=<rot 0 -0.0003 90.4093>
  }]
 dyn_Hair_Back_06 [{
   shapeType=Capsule
   length=3.0606
   radius=5.7116
   shapeName=dyn_Hair_Back_06_capsule
   center=<2.6983 -3.7731 0.027>
   rotation=<rot 0 -0.0003 90.4093>
  }]
 dyn_Hair_Back_07 [{
   shapeType=Capsule
   length=4.5606
   radius=4.2116
   shapeName=dyn_Hair_Back_06_capsule
   center=<1.0351 -3.7731 0.027>
   rotation=<rot 0 -0.0003 90.4093>
  }]
 dyn_Hair_Back_08 [{
   shapeType=Capsule
   length=4.5606
   radius=4.2116
   shapeName=dyn_Hair_Back_06_capsule
   center=<1.0351 -3.7731 0.027>
   rotation=<rot 0 -0.0003 90.4093>
  }]
 dyn_Hair_Back_09 [{
   shapeType=Capsule
   length=5.266
   radius=4.1996
   shapeName=dyn_Hair_Back_09_capsule
   center=<1.0351 -3.7731 0.027>
   rotation=<rot 0 -0.0003 90.4093>
  }]
 dyn_Hair_Back_10 [{
   shapeType=Capsule
   length=5.3349
   radius=4
   shapeName=dyn_Hair_Back_10_capsule
   center=<1.0351 -3.7731 0.027>
   rotation=<rot 0 -0.0003 90.4093>
  }]
 dyn_Hair_Back_11 [{
   shapeType=Capsule
   length=5.2571
   radius=4
   shapeName=dyn_Hair_Back_11_capsule
   center=<1.0351 -3.7731 0.027>
   rotation=<rot 0 -0.0003 90.4093>
  }]
 dyn_Hair_Back_12 [{
   shapeType=Capsule
   length=11.1777
   radius=4
   shapeName=dyn_Hair_Back_12_capsule
   center=<1.0351 -3.7731 0.027>
   rotation=<rot 0 -0.0003 90.4093>
  }]
 dyn_Hair_Back_13 [{
   shapeType=Capsule
   length=10.0972
   radius=4
   shapeName=dyn_Hair_Back_13_capsule
   center=<1.0351 -3.7731 0.027>
   rotation=<rot 0 -0.0003 90.4093>
  }]
 dyn_Hair_Back_14 [{
   shapeType=Capsule
   length=8.4042
   radius=4
   shapeName=dyn_Hair_Back_14_capsule
   center=<1.0351 -3.7731 0.027>
   rotation=<rot 0 -0.0003 90.4093>
  }]
 dyn_Hair_Back_15 [{
   shapeType=Capsule
   length=5.9286
   radius=4
   shapeName=dyn_Hair_Back_15_capsule
   center=<1.0351 -3.6568 0.0261>
   rotation=<rot 0 -0.0003 90.4093>
  }]
 dyn_Hair_Back_15a [{
   shapeType=Capsule
   length=2.7926
   radius=2.0412
   shapeName=dyn_Hair_Back_15a_capsule
   center=<2.6559 0.1824 0.3048>
   rotation=<rot 0 -0.2177 -88.741>
  }]
 dyn_Hair_Back_15b [{
   shapeType=Capsule
   length=6.3336
   radius=3.04
   shapeName=dyn_Hair_Back_15b_capsule
   center=<6.3995 -0.3331 -0.1737>
   rotation=<rot 0 -91.3301 -89.2356>
  }]
 dyn_Hair_Back_15b1 [{
   shapeType=Capsule
   length=5.5
   radius=3.1239
   shapeName=dyn_Hair_Back_15b1_capsule
   center=<6.6459 -0.5113 -0.4626>
   rotation=<rot 0 90.0642 -87.673>
  }]
 dyn_Hair_Back_16 [{
   shapeType=Capsule
   length=5.244
   radius=3
   shapeName=dyn_Hair_Back_16_capsule
   center=<0.4527 -5.2545 0.0375>
   rotation=<rot 0 -0.0003 90.4093>
  }]
 dyn_earring_l [{
   shapeType=Capsule
   length=3.5456
   radius=0.5
   shapeName=dyn_earring_l_capsule
   center=<2.9203 -0.0744 0.0546>
   rotation=<rot 0 -89.4684 -88.5292>
  }]
 dyn_earring_r [{
   shapeType=Capsule
   length=3.5456
   radius=0.5
   shapeName=dyn_earring_r_capsule
   center=<-2.9202 0.0744 -0.0546>
   rotation=<rot 0 -89.4684 -88.5292>
  }]
 head [{shapeType=Capsule, length=26.0416, radius=17.9176, shapeName=head_capsule, center=<7.2576 10.9598 -0.033>, rotation=<rot -0.3902 -0.0011 0>}]
 lowerarm_l [{shapeType=Capsule, length=36.5, radius=5.5, shapeName=lowerarm_l_capsule, center=<18.2717 0 0.7738>, rotation=<rot -0.0001 -89.9999 87.5751>}]
 lowerarm_r [{
   shapeType=Capsule
   length=36.5
   radius=5.5
   shapeName=lowerarm_r_capsule
   center=<-17.8687 0 -0.7567>
   rotation=<rot -0.0001 -89.9999 87.5751>
  }]
 neck_01 [{shapeType=Capsule, length=0.505, radius=0.505, shapeName=neck_01_capsule, rotation=<rot 0 -90 0>}]
 neck_02 [{shapeType=Capsule, length=6.1222, radius=6.4941, shapeName=neck_02_capsule, center=<7.6937 1.9129 0>, rotation=<rot 90 -69.9613 0>}]
 pelvis [{shapeType=Capsule, length=13.365, radius=16.635, shapeName=pelvis_capsule, center=<-3.0591 5.0782 0>}]
 root [{shapeType=Capsule, length=78.5, radius=163, shapeName=root_capsule, center=<-0.0047 -18.1958 -158.2446>, rotation=<rot -90 -0.015 0>}]
 spine_02 [{shapeType=Capsule, length=19.8569, radius=14.2801, shapeName=spine_02_capsule, center=<0 6.6409 0>, rotation=<rot 0 -90 0>}]
 spine_04 [{shapeType=Capsule, length=18.8883, radius=15.2486, shapeName=spine_04_capsule, center=<0 8.6198 0>, rotation=<rot 0 -90 0>}]
 spine_05 [{shapeType=Capsule, length=25.392, radius=21.7919, shapeName=spine_05_capsule, center=<7.8471 17.2181 0>, rotation=<rot 0 -90 0>}]
 thigh_l [{shapeType=Capsule, length=35.5, radius=6.5, shapeName=thigh_l_capsule, center=<-19.3244 0.0001 -0.8183>, rotation=<rot 0 -90.0002 87.5752>}]
 thigh_r [{shapeType=Capsule, length=35.5, radius=6.5, shapeName=thigh_r_capsule, center=<18.7878 0 0.7956>, rotation=<rot 0 -90.0002 87.5752>}]
 upperarm_l [{
   shapeType=Capsule
   length=33.4871
   radius=5.2452
   shapeName=upperarm_l_capsule
   center=<15.142 0.0017 0.2617>
   rotation=<rot -0.0074 -89.9934 -90.99>
  }]
 upperarm_r [{
   shapeType=Capsule
   length=35.0128
   radius=5.2653
   shapeName=upperarm_r_capsule
   center=<-15.7143 0 0.3059>
   rotation=<rot -0.0001 -89.9999 91.1152>
  }]
}
constraints [
 {
  bone1Name=head
  bone2Name=neck_02
  swing1LimitDegrees=45
  swing1Motion=Free
  swing2LimitDegrees=45
  swing2Motion=Free
  twistLimitDegrees=45
  twistMotion=Free
 }
 {
  bone1Name=spine_05
  bone2Name=spine_04
  swing1LimitDegrees=45
  swing1Motion=Free
  swing2LimitDegrees=45
  swing2Motion=Free
  twistLimitDegrees=45
  twistMotion=Free
 }
 {
  bone1Name=spine_04
  bone2Name=spine_02
  swing1LimitDegrees=45
  swing1Motion=Free
  swing2LimitDegrees=45
  swing2Motion=Free
  twistLimitDegrees=45
  twistMotion=Free
 }
 {
  bone1Name=clavicle_l
  bone2Name=spine_05
  swing1LimitDegrees=45
  swing1Motion=Free
  swing2LimitDegrees=45
  swing2Motion=Free
  twistLimitDegrees=45
  twistMotion=Free
 }
 {
  bone1Name=upperarm_l
  bone2Name=clavicle_l
  swing1LimitDegrees=45
  swing1Motion=Free
  swing2LimitDegrees=45
  swing2Motion=Free
  twistLimitDegrees=45
  twistMotion=Free
 }
 {bone1Name=dyn_Hair_Back_01, bone2Name=head, swing1LimitDegrees=45, swing2LimitDegrees=1, twistLimitDegrees=1}
 {bone1Name=dyn_Hair_Back_02, bone2Name=dyn_Hair_Back_01, swing1LimitDegrees=25, swing2LimitDegrees=1, twistLimitDegrees=1}
 {bone1Name=dyn_Hair_Back_03, bone2Name=dyn_Hair_Back_02, swing1LimitDegrees=30, swing2LimitDegrees=0, twistLimitDegrees=0}
 {bone1Name=dyn_Hair_Back_04, bone2Name=dyn_Hair_Back_03, swing1LimitDegrees=40, swing2LimitDegrees=0, twistLimitDegrees=0}
 {bone1Name=dyn_Hair_Back_05, bone2Name=dyn_Hair_Back_04, swing1LimitDegrees=40, swing2LimitDegrees=0, twistLimitDegrees=0}
 {bone1Name=dyn_Hair_Back_06, bone2Name=dyn_Hair_Back_05, swing1LimitDegrees=45, swing2LimitDegrees=3, twistLimitDegrees=1}
 {bone1Name=dyn_Hair_Back_07, bone2Name=dyn_Hair_Back_06, swing1LimitDegrees=45, swing2LimitDegrees=3, twistLimitDegrees=1}
 {bone1Name=dyn_Hair_Back_08, bone2Name=dyn_Hair_Back_07, swing1LimitDegrees=45, swing2LimitDegrees=3, twistLimitDegrees=1}
 {
  bone1Name=neck_01
  bone2Name=spine_05
  swing1LimitDegrees=45
  swing1Motion=Free
  swing2LimitDegrees=45
  swing2Motion=Free
  twistLimitDegrees=45
  twistMotion=Free
 }
 {
  bone1Name=clavicle_r
  bone2Name=spine_05
  swing1LimitDegrees=45
  swing1Motion=Free
  swing2LimitDegrees=45
  swing2Motion=Free
  twistLimitDegrees=45
  twistMotion=Free
 }
 {
  bone1Name=upperarm_r
  bone2Name=clavicle_r
  swing1LimitDegrees=45
  swing1Motion=Free
  swing2LimitDegrees=45
  swing2Motion=Free
  twistLimitDegrees=45
  twistMotion=Free
 }
 {bone1Name=dyn_Hair_Back_09, bone2Name=dyn_Hair_Back_08, swing1LimitDegrees=45, swing2LimitDegrees=3, twistLimitDegrees=1}
 {bone1Name=dyn_Hair_Back_10, bone2Name=dyn_Hair_Back_09, swing1LimitDegrees=55, swing2LimitDegrees=3, twistLimitDegrees=1}
 {bone1Name=dyn_Hair_Back_11, bone2Name=dyn_Hair_Back_10, swing1LimitDegrees=55, swing2LimitDegrees=3, twistLimitDegrees=1}
 {bone1Name=dyn_Hair_Back_12, bone2Name=dyn_Hair_Back_11, swing1LimitDegrees=55, swing2LimitDegrees=3, twistLimitDegrees=0.1}
 {bone1Name=dyn_Hair_Back_13, bone2Name=dyn_Hair_Back_12, swing1LimitDegrees=55, swing2LimitDegrees=3, twistLimitDegrees=0.1}
 {bone1Name=dyn_Hair_Back_14, bone2Name=dyn_Hair_Back_13, swing1LimitDegrees=55, swing2LimitDegrees=3, twistLimitDegrees=0.1}
 {bone1Name=dyn_Hair_Back_15, bone2Name=dyn_Hair_Back_14, swing1LimitDegrees=55, swing2LimitDegrees=3, twistLimitDegrees=0.1}
 {bone1Name=dyn_Hair_Back_16, bone2Name=dyn_Hair_Back_15, swing1LimitDegrees=5, swing2LimitDegrees=1, twistLimitDegrees=0.1}
 {
  bone1Name=lowerarm_r
  bone2Name=upperarm_r
  swing1LimitDegrees=45
  swing1Motion=Free
  swing2LimitDegrees=45
  swing2Motion=Free
  twistLimitDegrees=45
  twistMotion=Free
 }
 {
  bone1Name=lowerarm_l
  bone2Name=upperarm_l
  swing1LimitDegrees=45
  swing1Motion=Free
  swing2LimitDegrees=45
  swing2Motion=Free
  twistLimitDegrees=45
  twistMotion=Free
 }
 {
  bone1Name=thigh_l
  bone2Name=pelvis
  swing1LimitDegrees=45
  swing1Motion=Free
  swing2LimitDegrees=45
  swing2Motion=Free
  twistLimitDegrees=45
  twistMotion=Free
 }
 {
  bone1Name=calf_l
  bone2Name=thigh_l
  swing1LimitDegrees=45
  swing1Motion=Free
  swing2LimitDegrees=45
  swing2Motion=Free
  twistLimitDegrees=45
  twistMotion=Free
 }
 {
  bone1Name=thigh_r
  bone2Name=pelvis
  swing1LimitDegrees=45
  swing1Motion=Free
  swing2LimitDegrees=45
  swing2Motion=Free
  twistLimitDegrees=45
  twistMotion=Free
 }
 {
  bone1Name=calf_r
  bone2Name=thigh_r
  swing1LimitDegrees=45
  swing1Motion=Free
  swing2LimitDegrees=45
  swing2Motion=Free
  twistLimitDegrees=45
  twistMotion=Free
 }
 {bone1Name=dyn_Hair_Back_15a, bone2Name=dyn_Hair_Back_15, swing1LimitDegrees=5, swing2LimitDegrees=25, twistLimitDegrees=0.1}
 {bone1Name=dyn_Hair_Back_15b, bone2Name=dyn_Hair_Back_15, swing1LimitDegrees=5, swing2LimitDegrees=25, twistLimitDegrees=0.1}
 {bone1Name=dyn_Hair_Back_15b1, bone2Name=dyn_Hair_Back_15b, swing1LimitDegrees=5, swing2LimitDegrees=25, twistLimitDegrees=0.1}
 {
  bone1Name=pelvis
  bone2Name=root
  swing1LimitDegrees=45
  swing1Motion=Free
  swing2LimitDegrees=45
  swing2Motion=Free
  twistLimitDegrees=45
  twistMotion=Free
 }
 {
  bone1Name=spine_02
  bone2Name=root
  swing1LimitDegrees=45
  swing1Motion=Free
  swing2LimitDegrees=45
  swing2Motion=Free
  twistLimitDegrees=45
  twistMotion=Free
 }
 {
  bone1Name=base_earring_l
  bone2Name=head
  swing1LimitDegrees=45
  swing1Motion=Free
  swing2LimitDegrees=45
  swing2Motion=Free
  twistLimitDegrees=45
  twistMotion=Free
 }
 {bone1Name=dyn_earring_l, bone2Name=base_earring_l, swing1LimitDegrees=35, swing2LimitDegrees=35, twistLimitDegrees=45, twistMotion=Locked}
 {
  bone1Name=base_earring_r
  bone2Name=head
  swing1LimitDegrees=45
  swing1Motion=Free
  swing2LimitDegrees=45
  swing2Motion=Free
  twistLimitDegrees=45
  twistMotion=Free
 }
 {bone1Name=dyn_earring_r, bone2Name=base_earring_r, swing1LimitDegrees=30, swing2LimitDegrees=30, twistLimitDegrees=45, twistMotion=Locked}
 {
  bone1Name=neck_02
  bone2Name=neck_01
  swing1LimitDegrees=45
  swing1Motion=Free
  swing2LimitDegrees=45
  swing2Motion=Free
  twistLimitDegrees=45
  twistMotion=Free
 }
]
