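# Post-process-style AnimBlueprint for the TurnVortex female body's FaceAcc part, adding secondary physics over the pose copied from the character mesh.
# A constrained gravity bone and Control Rig feed a cached pose into three RigidBody simulations with separate physics assets.
# Full dynamics give way to pack or cape variants during skydiving, DBNO, backbling wear or TurnVortex emotes.
# Gravity and simulation settings are computed using LoanFloat's borrowed parameter asset.
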AnimBlueprint /BRCosmetics/Characters/Player/Female/Medium/Bodies/F_MED_TurnVortex/Meshes/Parts/F_MED_TurnVortex_FaceAcc_AnimBP
target_skeleton=&/F_MED_TurnVortex_FaceAcc_Skeleton
graphs [
 AnimGraph {nodes [
   n1 = RigidBody!(
    ComponentPose=Local To Component(LocalPose=Use cached pose 'HairCachePose'())
    Alpha
    OverrideWorldGravity:Vector="0,0,0"
    SimSpaceSettings:Sim Space Settings Structure=(WorldAlpha=0.1,VelocityScaleZ=0,DampingAlpha=1,MaxLinearVelocity=500,MaxAngularVelocity=10,MaxLinearAcceleration=700,MaxAngularAcceleration=1000,ExternalLinearDragV=(X=0.5,Y=0.5,Z=0),ExternalLinearVelocity=(X=0,Y=0,Z=0),ExternalAngularVelocity=(X=0,Y=0,Z=0))
   ) -> (Pose) {
    bAllowPoseScalePropagation=false
    bClampLinearTranslationLimitToRefPose=true
    bOverrideWorldGravity=true
    overridePhysicsAsset=&/F_MED_TurnVortex_FaceAcc_WearingPackBool_Physics
    simulationSpace=BaseBoneSpace
    alphaBoolBlend {blendOption=Linear}
    baseBoneRef {boneName=gravity}
    componentAppliedLinearAccClamp=<0 0 0>
    overrideWorldGravity {z=-980}
    simSpaceSettings {
     maxAngularAcceleration=1000
     maxAngularVelocity=10
     maxLinearAcceleration=700
     maxLinearVelocity=500
     velocityScaleZ=0
     worldAlpha=0.1
     externalLinearDragV=<0.5 0.5>
    }
   }
   n2 = Component To Local(ComponentPose=@n1)
   n3 = RigidBody!(ComponentPose=Local To Component(LocalPose=Use cached pose 'HairCachePose'()), Alpha=0.7, OverrideWorldGravity="0,0,0") -> (Pose) {
    bAllowPoseScalePropagation=false
    bClampLinearTranslationLimitToRefPose=true
    bOverrideWorldGravity=true
    overridePhysicsAsset=&/F_MED_TurnVortex_FaceAcc_Cape_Physics
    simulationSpace=BaseBoneSpace
    alphaBoolBlend {blendOption=Linear}
    baseBoneRef {boneName=gravity}
    componentAppliedLinearAccClamp=<0 0 0>
    overrideWorldGravity {z=-980}
    simSpaceSettings {
     dampingAlpha=0.5
     maxAngularAcceleration=80
     maxAngularVelocity=0.5
     maxLinearAcceleration=10
     maxLinearVelocity=10
     worldAlpha=1
     externalLinearDragV=<1 1 1>
    }
   }
   n4 = Component To Local(ComponentPose=@n3)
   n5 = Get isWearingCape() -> (isWearingCape:bool)
   Output Pose(Result=Blend Poses by bool(
     BlendPose_0=Blend Poses by bool(BlendPose_0=@n4, BlendPose_1=@n2, bActiveValue=@n5) {blendType=Linear}
     BlendPose_1=Use cached pose 'HairDynPose'()
     bActiveValue=Get disableHalfDynamics() -> (disableHalfDynamics:bool)
    ) {blendType=Linear}) {layerGroup=None}
   n6 = Get Character Mesh() -> (Character Mesh:Skeletal Mesh Component Object Reference)
   n7 = "Constraint - gravity (TT)"!(
    ComponentPose=Local To Component(LocalPose=Copy Pose From Mesh(SourceMeshComponent=@n6) {bUseAttachedParent=true})
    Alpha=1
    ConstraintWeights_0=0.2
    ConstraintWeights_1:float=0.8
   ) -> (Pose) {
    alphaBoolBlend {blendOption=Linear}
    boneToModify {boneName=gravity}
    constraintSetup [{transformType=Translation, targetBone {boneName=root}}, {transformType=Translation, targetBone {boneName=head}}]
    constraintWeights [1,1]
   }
   n8 = "Constraint - gravity (PP)"!(ComponentPose=@n7, Alpha=1, ConstraintWeights_0=1, ConstraintWeights_1:float=0) -> (Pose) {
    alphaBoolBlend {blendOption=Linear}
    boneToModify {boneName=gravity}
    constraintSetup [{targetBone {boneName=root}}, {targetBone {boneName=pelvis}}]
    constraintWeights [1,1]
   }
   n9 = Control Rig(Source=Component To Local(ComponentPose=@n8)) {
    alphaBoolBlend {blendOption=Linear}
    controlRigAssetReference {blueprintRigClass=&/F_MED_TurnVortex_FaceAcc_CtrlRig.F_MED_TurnVortex_FaceAcc_CtrlRig_C}
   }
   Save cached pose 'HairCachePose'(Pose=@n9) {cacheName=HairCachePose}
   n10 = RigidBody!(
    ComponentPose=Local To Component(LocalPose=Use cached pose 'HairCachePose'())
    Alpha
    OverrideWorldGravity="0,0,0"
    SimSpaceSettings:Sim Space Settings Structure=(WorldAlpha=0.1,VelocityScaleZ=0,DampingAlpha=1,MaxLinearVelocity=500,MaxAngularVelocity=10,MaxLinearAcceleration=700,MaxAngularAcceleration=1000,ExternalLinearDragV=(X=0.5,Y=0.5,Z=0),ExternalLinearVelocity=(X=0,Y=0,Z=0),ExternalAngularVelocity=(X=0,Y=0,Z=0))
   ) -> (Pose) {
    bAllowPoseScalePropagation=false
    bClampLinearTranslationLimitToRefPose=true
    bOverrideWorldGravity=true
    overridePhysicsAsset=&/F_MED_TurnVortex_FaceAcc_Physics
    simulationSpace=BaseBoneSpace
    alphaBoolBlend {blendOption=Linear}
    baseBoneRef {boneName=gravity}
    componentAppliedLinearAccClamp=<0 0 0>
    overrideWorldGravity {z=-980}
    simSpaceSettings {
     maxAngularAcceleration=1000
     maxAngularVelocity=10
     maxLinearAcceleration=700
     maxLinearVelocity=500
     velocityScaleZ=0
     worldAlpha=0.1
     externalLinearDragV=<0.5 0.5>
    }
   }
   Save cached pose 'HairDynPose'(Pose=Component To Local(ComponentPose=@n10)) {cacheName=HairDynPose}
  ]}
 EventGraph {nodes [
   Event BlueprintUpdateAnimation()
   > n1 = Sequence()
   > n2 = "Gravity Override State Function (Parameter Asset) (Nativized)"(
    CharacterPartAnimInstance=Self-Reference()
    GameTimeInSeconds=GetGameTimeInSeconds()
    Parameters=/BRCosmetics/Characters/Player/Female/Medium/Bodies/F_MED_LoanFloat/Meshes/Parts/F_MED_LoanFloat_FaceAcc_Parameters
   )
   > Set Gravity Override(Gravity Override=@n2)
   > Set RBAN Settings Out(RBAN Settings Out=@n2.OutRigidBodyAnimNodeParameters)
   n3 = GetAnimInstance(self=GetOwnerSkeletalMeshComponent())
   > Set isPlayingTurnVortexEmotePack(isPlayingTurnVortexEmotePack:bool=OR Boolean!(
     A=Montage_IsPlaying(self=@n3, Montage=/BRCosmetics/Animation/Game/MainPlayer/Emotes/TurnVortex/Leader/CLF/Emote_TurnVortex_Intro_CLF_M)
     B=Montage_IsPlaying(self=@n3, Montage=/BRCosmetics/Animation/Game/MainPlayer/Emotes/TurnVortex/Leader/CMF/Emote_TurnVortex_LeaderSync_CMF_M)
     C:bool=Montage_IsPlaying(self=@n3, Montage=/BRCosmetics/Animation/Game/MainPlayer/Emotes/TurnVortex/Joiner/CMF/Emote_TurnVortex_Joiner_CMF_M)
    ) -> (ReturnValue)) -> (Output_Get:bool)
   Set disableHalfDynamics(
    execute=@n1.then_1
    disableHalfDynamics:bool=OR Boolean!(
     A=Get bIsSkydiveDiveMode()
     B=Get bIsSkydiveFloating()
     C=Get isWearingPack() -> (isWearingPack:bool)
     D:bool=Get bIsDBNO()
     E:bool=Get bIsDBNOCarrying() -> (bIsDBNOCarrying:bool)
     F:bool=Get isPlayingTurnVortexEmotePack() -> (isPlayingTurnVortexEmotePack:bool)
    ) -> (ReturnValue)
   ) -> (Output_Get:bool)
   n4 = Event OnCharacterCustomizationLoaded()
   > Set Character Mesh(Character Mesh:Skeletal Mesh Component Object Reference=@n4.CustomBaseMesh) -> (Output_Get:Skeletal Mesh Component Object Reference)
   > Set Head_Mesh(Head_Mesh:Skeletal Mesh Component Object Reference=Get Array Item with EFort Custom Part Type(Array=@n4.CustomMeshParts, PartType=Head)) -> (Output_Get:Skeletal Mesh Component Object Reference)
   n5 = CheckForBackbling()
   > Set isWearingPack(isWearingPack:bool=OR Boolean!(A=@n5, B=@n5.bIsBackblingACape, C=@n5.bIsBackblingLong) -> (ReturnValue)) -> (Output_Get:bool)
   > Set isWearingCape(isWearingCape=@n5.bIsBackblingLong)
  ]}
]
variables [
 Character Mesh {type=SkeletalMeshComponent}
 disableHalfDynamics {type=boolean}
 Gravity Override {type=Vector}
 Head_Mesh {type=SkeletalMeshComponent}
 isPlayingTurnVortexEmotePack {type=boolean}
 isWearingCape {type=boolean}
 isWearingPack {type=boolean}
 RBAN Settings Out {type=FortRigidBodyAnimNodeParameters}
]
